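; Cosmetic part modifier for the SnapFreeze Graze outfit that drives reactive damage and elimination effects.
; Damage events convert the stored hit location to mesh space, with a fallback, then write position and timestamp into round-robin material parameters and pulse a Niagara prototype flag.
; Kills spawn an attached Niagara burst and timestamp materials when the first emissive variant is active, replayed on delays in the front end.
; Setup passes body/head and material variant to idle VFX.
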(event EventOnPostResIn
  ;[0 0 0] on post res-in
  (|CallParentFunction)
  (Effects|Components|ParticleSystem|SetFloatParameter (CallFunction|GetIdleVFXComponent) "CharacterPart" (Variables|Default|GetBodyOrHead))
  (Default|VariantSetup)
  (if (CustomCosmeticModifier|IsinFrontEnd)
    (Utilities|FlowControl|Delay 5.0)
    (CallFunction|KillReactivity)
    (Utilities|FlowControl|Delay 7.5)
    (CallFunction|KillReactivity)))

(event EventOnStatChanged
  (params (State%20Name :type Name :default "None") (Stat%20Value :type Integer))
  ;[0 0 0] on stat changed
  (|CallParentFunction State%20Name Stat%20Value)
  (if (> Stat%20Value 0)
    _))

(event EventOnPawnAndPartChanged
  (bind _returnvalue (CustomCharacter|GetSkeletalMeshForPartOnPlayerPawn))
  (bind _returnvalue_2 (Transformation|GetWorldTransform _returnvalue))
  (bind _rawimpactlocation (Variables|Default|GetRawImpactLocation))
  (bind _returnvalue_3 (Math|Transform|InverseTransformLocation _returnvalue_2 _rawimpactlocation))
  ; Hit React
  (|CallParentFunction)
  (Utilities|FlowControl|Sequence
    (:then_0
      (Cosmetics|WatchCosmeticStat (CustomCharacter|GetAssociatedPlayerPawn) "(TagName=\"StatManager.DamageReceived\")" 0 false
        (:OnStatChanged
          (if (or (>= (Math|Vector|VectorLength _returnvalue_3) 155.0) (Utilities|Operators|Equal(==) :A (Variables|Default|GetRawImpactLocation)))
            (CallFunction|SendImpactDatatoMaterials "0.000000,0.000000,75.000000")
            (else
              (bind _output_get (Variables|Default|SetImpactLocation _returnvalue_3))
              (CallFunction|SendImpactDatatoMaterials _output_get)))
          ;; Niagara Prototype
          (Math|Integer|IncrementInt (Variables|Default|GetImpactCounter)
            (:%20%20
              (Variables|Default|SetHasNewImpact true)
              (Utilities|FlowControl|DelayUntilNextTick)
              (Variables|Default|SetHasNewImpact false))))))
    (:then_1
      (Cosmetics|WatchCosmeticStat (CustomCharacter|GetAssociatedPlayerPawn) "(TagName=\"StatManager.AthenaKills\")" 0 false
        (:OnStatChanged
          (CallFunction|KillReactivity))))))

(event EventOnHitPawn
  (params (SelfActor :type Actor%20Object%20Reference) (OtherActor :type Actor%20Object%20Reference) (NormalImpulse :type Vector :default "0, 0, 0") (Hit :type Hit%20Result%20Structure%20%28by%20ref%29))
  (bind (_bblockinghit _binitialoverlap _time _distance _location) (Collision|BreakHitResult Hit))
  ; Hit React
  (|CallParentFunction SelfActor OtherActor NormalImpulse Hit)
  (Variables|Default|SetRawImpactLocation _location))

(fn KillReactivity
  (bind _asfort_player_pawn (Utilities|Casting|CastToFortPlayerPawn (Actor|GetOwner)))
  (bind _index (CosmeticLoadout|GetVariantChannelActiveIndexforSlotActor "(TagName=\"LoadoutSlot.Character\")" _asfort_player_pawn "(TagName=\"Cosmetics.Variant.Channel.Slot.0\")" (Utilities|Array|MakeArray "(TagName=\"Cosmetics.Variant.Property.Emissive0\")" "(TagName=\"Cosmetics.Variant.Property.Emissive1\")" "(TagName=\"Cosmetics.Variant.Property.Emissive2\")" "(TagName=\"Cosmetics.Variant.Property.Emissive3\")")))
  (switch int _index
    (:0
      (bind _returnvalue (PartModifier|SpawnSystem|SpawnSystemAttachedPartModifier "/BRCosmetics/Characters/Player/Female/Medium/Bodies/F_MED_SnapFreeze_Graze/FX/NS_Character_SnapFreeze_Graze.NS_Character_SnapFreeze_Graze" (CustomCharacter|GetSkeletalMeshForPartOnPlayerPawn))
        (:Success
          (Effects|Components|ParticleSystem|SetFloatParameter _returnvalue "CharacterPart" (Variables|Default|GetBodyOrHead))
          (Effects|Components|ParticleSystem|SetFloatParameter _returnvalue "GameTime" (Utilities|Time|GetGameTimeinSeconds))
          (Rendering|Material|SetScalarParameterValueonMaterials (CustomCharacter|GetSkeletalMeshForPartOnPlayerPawn) "Elim_StartTime" (Utilities|Time|GetGameTimeinSeconds))
          (Effects|Components|ParticleSystem|SetBoolParameter _returnvalue "bisHandHidden" (Variables|HiddenPartCheck|GetbIsHandHidden))
          (Effects|Components|ParticleSystem|SetBoolParameter _returnvalue "bisHeadHidden" (Variables|HiddenPartCheck|GetbIsHeadHidden)))))))

(fn SendImpactDataToMaterials
  (params (ImpactPosition :type Vector))
  (bind _returnvalue (CustomCharacter|GetSkeletalMeshForPartOnPlayerPawn))
  (bind _output_get (Variables|Default|SetLastImpactIndex (Math|Integer|%25%28Integer%29 (+ (Variables|Default|GetLastImpactIndex) 1) (Variables|Default|GetNumConcurrentImpacts))))
  (Rendering|Material|SetVectorParameterValueonMaterials _returnvalue (Utilities|String|StringToName (Utilities|String|Append "HitVector" (Utilities|String|ToString(Integer) _output_get))) ImpactPosition)
  (Rendering|Material|SetScalarParameterValueonMaterials _returnvalue (Utilities|String|StringToName (Utilities|String|Append "HitTimeStamp" (Utilities|String|ToString(Integer) _output_get))) (Utilities|Time|GetGameTimeinSeconds)))

(fn ConstructionScript
  (|CallParentFunction))

(fn VariantSetup
  (bind _asfort_player_pawn (Utilities|Casting|CastToFortPlayerPawn (Actor|GetOwner)))
  (bind _index (CosmeticLoadout|GetVariantChannelActiveIndexforSlotActor "(TagName=\"LoadoutSlot.Character\")" _asfort_player_pawn "(TagName=\"Cosmetics.Variant.Channel.Material\")" (Utilities|Array|MakeArray "(TagName=\"Cosmetics.Variant.Property.Mat1\")" "(TagName=\"Cosmetics.Variant.Property.Mat2\")")))
  (Effects|Components|ParticleSystem|SetFloatParameter (CallFunction|GetIdleVFXComponent) "Variant" (Math|Conversions|ToFloat(Integer) _index)))
