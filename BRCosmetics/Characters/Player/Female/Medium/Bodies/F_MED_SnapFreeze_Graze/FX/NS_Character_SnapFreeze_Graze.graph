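# Niagara cosmetic FX for the SnapFreeze Graze character, pinning persistent glows to skeletal mesh sockets.
# User.CharacterPart drives burst probability, selecting either BodyGlow and BraceletGlow sprites with waveform flicker, or the Head emitter's two eye glows and lights.
# Head glows follow per-eye blink values, and hidden-head or hidden-hand flags control sprite size or visibility tag.
# Variant selectors set colors, with time-of-day and front-end tints on some layers, while system ReactivityTiming scales many glows.
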NiagaraSystem /BRCosmetics/Characters/Player/Female/Medium/Bodies/F_MED_SnapFreeze_Graze/FX/NS_Character_SnapFreeze_Graze
dependencies {
 usedDataInterfaces [
  &NiagaraDataInterfaceArrayFloat3
  &NiagaraDataInterfaceArrayFloat
  &NiagaraDataInterfaceSkeletalMesh
  &NiagaraDataInterfaceCurve
 ]
 usedDynamicInputs [
  &SimulationPosition
  &SelectFloatFromArray
  &FloatFromCurve
  &ClampFloat
  &Divide_Float
  &Subtract_Float
  &RemapRange
  &DotProduct
  &ReturnMeshOrientationAxis
  &FNTimeOfDayScalar
 ]
 usedModules [
  &self:SystemSpawnScript.SystemScriptSource.SystemScriptGraph.NiagaraNodeAssignment_9.SetVariables_#1
  &self:SystemSpawnScript.SystemScriptSource.SystemScriptGraph.NiagaraNodeAssignment_3.SetVariables_#2
  &self:SystemSpawnScript.SystemScriptSource.SystemScriptGraph.NiagaraNodeAssignment_1.SetVariables_#3
  &SystemState
  &FN_VariantManager
  &self:SystemSpawnScript.SystemScriptSource.SystemScriptGraph.NiagaraNodeAssignment_2.SetVariables_#4
  &SampleSkeletalMesh
  &self:SystemSpawnScript.SystemScriptSource.SystemScriptGraph.NiagaraNodeAssignment_4.SetVariables_#5
  &self:SystemSpawnScript.SystemScriptSource.SystemScriptGraph.NiagaraNodeAssignment_5.SetVariables_#6
 ]
}
emitters {
 BodyGlow {
  emitterUpdateScript [
   &EmitterState(
    Inactive Response=::NewEnumerator1
    Life Cycle Mode=::NewEnumerator1
    Loop Behavior=::NewEnumerator1
    Loop Duration Mode=::NewEnumerator1
    UseLoopDelay=false
   )
   &SpawnBurst_Instantaneous(
    Spawn Count=13
    Spawn Probability=&Lerp_Float(A=1, B=0, Alpha=$User.CharacterPart)
    Use Spawn Probability=true
    UseLoopCountLimit=true
   )
   /Niagara/Noise/Waveform(
    EnableExponent1=false
    GlobalClampMax=false
    GlobalClampMin=false
    UseInput1=false
    UsePhase1=false
    "[1] Amplitude Min/Max":Vector2f=<0.4 1>
    "[1] Frequency"=1.5
   )
   /SetVariables(
    Emitter.Flicker:NiagaraFloat=&Add_Float(A=&Multiply_Float(A=$Emitter.Waveform.Waveform, B=0.5), B=0.7)
    Emitter.FlickerHeavy:NiagaraFloat=&Add_Float(A=&Multiply_Float(A=$Emitter.Waveform.Waveform, B=1), B=0.8)
   ) {isSetParameters=true, module=SetVariables_#7}
   /SetVariables(
    Emitter.CoreGlowColor:LinearColor=&MultiplyLinearColor(
     Linear Color=&VariantColorSelector(
      Variant Amount=0
      Original Color=<rgba 0.2268 0 1 1>
      Variant 1 Color=<rgba 0.8855 0.3245 1 1>
      Variant 2 Color=?
      Use FrontEnd?=true
      Front End Color=<rgba 1 1 1 1>
      Gameplay Color=<rgba 1 1 1 1>
     )
     Scale Factor=&MakeVector4_FromFloat(Float=&Multiply_Float(A=1, B=$Emitter.Flicker))
    )
    Emitter.CoreGlowSize:Vector2f=&Vector2DFromFloat(Value=2)
    Emitter.DarkGlowColor:LinearColor=&MultiplyLinearColor(
     Linear Color=&MultiplyLinearColor(
      Linear Color=&VariantColorSelector(
       Variant Amount=0
       Morning Color=<rgba 0.8196 0.4824 1 1>
       Day Color=<rgba 1 1 1 1>
       Evening Color=<rgba 0.7714 0.5883 0.974 1>
       Night Color=<rgba 0.3984 0.5 1 1>
       Use TOD?=true
       Original Color=<rgba 0.0843 0.0199 0.1812 1>
       Variant 1 Color=<rgba 0.1349 0.0634 0.1812 1.5>
       Variant 2 Color=?
       Use FrontEnd?=true
       Front End Color=<rgba 1 1 1 1>
       Gameplay Color=<rgba 1 1 1 1>
      )
      Scale Factor=&MakeVector4_FromFloat(Float=$System.ReactivityTiming)
     )
     Scale Factor=&MakeVector4_FromFloat(Float=&Multiply_Float(A=1, B=$Emitter.Flicker))
    )
    Emitter.DarkGlowSize:Vector2f=&Vector2DFromFloat(Value=8.0622)
    Emitter.DepthFade=1
    Emitter.DiffuseGlowColor:LinearColor=&MultiplyLinearColor(
     Linear Color=&VariantColorSelector(
      Variant Amount=0
      Morning Color=<rgba 0.8196 0.4824 1 1>
      Day Color=<rgba 1 1 1 1>
      Evening Color=<rgba 0.5246 0.3489 0.974 1>
      Night Color=<rgba 0.3984 0.5 1 1>
      Use TOD?=true
      Original Color=<rgba 0.6511 0.3134 2 1>
      Variant 1 Color=<rgba 0.2193 0.3415 4 1>
      Variant 2 Color=?
      Use FrontEnd?=true
      Front End Color=<rgba 0.7 0.7 0.7 1>
      Gameplay Color=<rgba 1 1 1 1>
     )
     Scale Factor=&MakeVector4_FromFloat(Float=&Multiply_Float(A=1, B=$Emitter.Flicker))
    )
    Emitter.EyeGlowCameraOffset:NiagaraFloat=&FN_FrontEndFloatModifier(Gameplay=7, Front End=1)
    Emitter.LightExponent=5
    Emitter.LightGlowColor:LinearColor=&MultiplyLinearColor(
     Linear Color=&VariantColorSelector(Variant Amount=0, Original Color=<rgba 1.5 0.2154 0 1>, Variant 1 Color=<rgba 0.0352 0.1636 1 1>, Variant 2 Color=?)
     Scale Factor=&MakeVector4_FromFloat(Float=&Multiply_Float(A=7, B=$Emitter.FlickerHeavy))
    )
    Emitter.LightRadius=3
   ) {isSetParameters=true, module=SetVariables_#8}
  ]
  particleSpawnScript [
   &PartitionParticles(Number Of Partitions=13)
   &PartitionParticles(Particles Per Partition=6, Partition Mode:ENiagara_PartitionMode=::NewEnumerator1) {module=PartitionParticles001}
   &InitializeParticle(Color=<rgba 1 0.3963 0 0.7>, Sprite Size Mode=::NewEnumerator3, Uniform Sprite Size=120, UsePositionOffset=false, Write Material Random=true)
   /SetVariables(Particles.isSmallBead:NiagaraBool=&Boolean_OrOperation(
     Bool1=&SetBoolByIntComparison(A=$Particles.PartitionParticles001.Index In Partition, B=2, Comparison Type=::NewEnumerator4)
     Bool2=&Boolean_OrOperation(
      Bool1=&SetBoolByIntComparison(A=$Particles.PartitionParticles001.Index In Partition, B=5, Comparison Type=::NewEnumerator4)
      Bool2=&SetBoolByIntComparison(A=$Particles.PartitionParticles001.Index In Partition, B=4, Comparison Type=::NewEnumerator4)
     )
    )) {isSetParameters=true}
  ]
  particleUpdateScript [
   &ParticleState(Kill Particles When Lifetime Has Elapsed=false, Let Infinitely Lived Particles Die When Emitter Deactivates=true)
   &SkeletalMeshLocation(
    Bone / Socket Index=$Particles.PartitionParticles.Partition
    MaskSpawnGroup=false
    Mesh Sampling Type=::NewEnumerator5
    Sampled Position Offset=&FN_FrontEndVector3Modifier(
     Gameplay=&SelectVectorFromArray(
      Direct Array Index=$Particles.PartitionParticles.Partition
      Vector Selection Array=DataInterface() {propertyValues {FloatData [
         <0 0 0.15>
         <0 0 0>
         <0.2 -0.8 -0.5>
         <0 0 0>
         <0 0 -0.3>
         <0 0 0>
         <0 0 0>
         <0 0 0>
         <0.7 0 0>
         <0 0 0>
         <0 0.1 -0.5>
         <0 0 0>
         <0 0 0>
         <0 0 0>
        ]}}
     )
     Front End=&SelectVectorFromArray(
      Direct Array Index=$Particles.PartitionParticles.Partition
      Vector Selection Array=DataInterface() {propertyValues {FloatData [
         <0 0 0.15>
         <0 0 0>
         <0.2 -0.8 -0.5>
         <0 0 0>
         <0 0 -0.3>
         <0 0 0>
         <0 0 0>
         <0 0 0>
         <0.7 0 0>
         <0 0 0>
         <0 0.1 -0.5>
         <0 0 0>
         <0 0 0>
         <0 0 0>
        ]}}
     )
    )
    Skeletal Mesh=$System.Skeletal Mesh
    Socket Sampling Mode=::NewEnumerator5
   )
   &ScaleColor(
    Scale Alpha=&Multiply_Float(A=$Emitter.Flicker, B=0.3)
    Scale RGB=&VectorFromFloat(Value=$System.TODFloat)
    ScaleA=true
    ScaleRGB=true
   )
   /SetVariables(Particles.DiffuseGlowSize:Vector2f=&Multiply_Vector2D_ByFloat(
     Float=$System.ReactivityTiming
     Vector2D=&Vector2DFromFloat(Value=&MakeCustomFloatFromBool(Bool=$Particles.isSmallBead, True Float=6, False Float=7))
    )) {isSetParameters=true}
  ]
  rendererClasses [&NiagaraSpriteRendererProperties]
  renderers [
   {rendererClass=&NiagaraSpriteRendererProperties}
   {rendererClass=&NiagaraSpriteRendererProperties, rendererIndex=1}
  ]
 }
 BraceletGlow {
  emitterUpdateScript [
   &EmitterState(
    Inactive Response=::NewEnumerator1
    Life Cycle Mode=::NewEnumerator1
    Loop Behavior=::NewEnumerator1
    Loop Duration Mode=::NewEnumerator1
    UseLoopDelay=false
   )
   &SpawnBurst_Instantaneous(
    Spawn Count=12
    Spawn Probability=&Lerp_Float(A=1, B=0, Alpha=$User.CharacterPart)
    Use Spawn Probability=true
    UseLoopCountLimit=true
   )
   &Waveform(
    EnableExponent1=false
    GlobalClampMax=false
    GlobalClampMin=false
    UseInput1=false
    UsePhase1=false
    "[1] Amplitude Min/Max":Vector2f=<0.4 1>
    "[1] Frequency"=1.5
   )
   /SetVariables(
    Emitter.Flicker:NiagaraFloat=&Add_Float(A=&Multiply_Float(A=$Emitter.Waveform.Waveform, B=0.5), B=0.7)
    Emitter.FlickerHeavy:NiagaraFloat=&Add_Float(A=&Multiply_Float(A=$Emitter.Waveform.Waveform, B=1), B=0.8)
   ) {isSetParameters=true, module=SetVariables_#7}
   /SetVariables(
    Emitter.CoreGlowColor:LinearColor=&MultiplyLinearColor(
     Linear Color=&VariantColorSelector(Variant Amount=0, Original Color=<rgba 0.2268 0 1 1>, Variant 1 Color=<rgba 0.8855 0.3245 1 1>, Variant 2 Color=?)
     Scale Factor=&MakeVector4_FromFloat(Float=&Multiply_Float(A=1, B=$Emitter.Flicker))
    )
    Emitter.CoreGlowSize:Vector2f=&Vector2DFromFloat(Value=2)
    Emitter.DarkGlowColor:LinearColor=&MultiplyLinearColor(
     Linear Color=&MultiplyLinearColor(
      Linear Color=&VariantColorSelector(
       Variant Amount=0
       Morning Color=<rgba 0.8196 0.4824 1 1>
       Day Color=<rgba 1 1 1 1>
       Evening Color=<rgba 0.7714 0.5883 0.974 1>
       Night Color=<rgba 0.2 0.251 0.502 1>
       Use TOD?=true
       Original Color=<rgba 0.0843 0.0199 0.1812 1>
       Variant 1 Color=<rgba 0.1349 0.0634 0.1812 1.5>
       Variant 2 Color=?
      )
      Scale Factor=&MakeVector4_FromFloat(Float=&Multiply_Float(A=1, B=$Emitter.Flicker))
     )
     Scale Factor=&MakeVector4_FromFloat(Float=$System.ReactivityTiming)
    )
    Emitter.DarkGlowSize:Vector2f=&Vector2DFromFloat(Value=8)
    Emitter.DepthFade=1
    Emitter.DiffuseGlowColor:LinearColor=&MultiplyLinearColor(
     Linear Color=&VariantColorSelector(
      Variant Amount=0
      Morning Color=<rgba 0.8196 0.4824 1 1>
      Day Color=<rgba 1 1 1 1>
      Evening Color=<rgba 0.7714 0.5883 0.974 1>
      Night Color=<rgba 0.2 0.251 0.502 1>
      Use TOD?=true
      Original Color=<rgba 0.6511 0.3134 2 1>
      Variant 1 Color=<rgba 0.2193 0.3415 4 1>
      Variant 2 Color=?
     )
     Scale Factor=&MakeVector4_FromFloat(Float=&Multiply_Float(A=$System.ReactivityTiming, B=$Emitter.Flicker))
    )
    Emitter.DiffuseGlowSize:Vector2f=&Vector2DFromFloat(Value=4)
    Emitter.LightExponent=5
    Emitter.LightGlowColor:LinearColor=&MultiplyLinearColor(
     Linear Color=&VariantColorSelector(Variant Amount=0, Original Color=<rgba 1.5 0.2154 0 1>, Variant 1 Color=<rgba 0.0352 0.1636 1 1>, Variant 2 Color=?)
     Scale Factor=&MakeVector4_FromFloat(Float=&Multiply_Float(A=7, B=$Emitter.FlickerHeavy))
    )
    Emitter.LightRadius=3
   ) {isSetParameters=true, module=SetVariables_#8}
  ]
  particleSpawnScript [
   &PartitionParticles(Number Of Partitions=12)
   &InitializeParticle(Color=<rgba 1 0.3963 0 0.7>, Sprite Size Mode=::NewEnumerator3, Uniform Sprite Size=120, UsePositionOffset=false, Write Material Random=true)
  ]
  particleUpdateScript [
   &ParticleState(Kill Particles When Lifetime Has Elapsed=false, Let Infinitely Lived Particles Die When Emitter Deactivates=true)
   &SkeletalMeshLocation(
    Bone / Socket Index=&ReturnExecIndex()
    MaskSpawnGroup=false
    Mesh Sampling Type=::NewEnumerator5
    Skeletal Mesh=$System.Skeletal Mesh Wrist
    Socket Sampling Mode=::NewEnumerator5
   )
   &ScaleColor(
    Scale Alpha=&Multiply_Float(A=&Multiply_Float(A=$Emitter.Flicker, B=0.3), B=$System.ReactivityTiming)
    Scale RGB=&VectorFromFloat(Value=$System.TODFloat)
    ScaleA=true
    ScaleRGB=true
   )
   /SetVariables(Particles.VisibilityTag=&MakeCustomIntFromBool(Bool=$User.bisHandHidden, True INT=1, False INT=0)) {isSetParameters=true}
  ]
  rendererClasses [&NiagaraSpriteRendererProperties]
  renderers [
   {rendererClass=&NiagaraSpriteRendererProperties}
   {rendererClass=&NiagaraSpriteRendererProperties, rendererIndex=1}
  ]
 }
 Head {
  emitterUpdateScript [
   &EmitterState(
    Inactive Response=::NewEnumerator1
    Life Cycle Mode=::NewEnumerator1
    Loop Behavior=::NewEnumerator1
    Loop Duration Mode=::NewEnumerator1
    UseLoopDelay=false
   )
   &SpawnBurst_Instantaneous(
    Spawn Count=2
    Spawn Probability=&Lerp_Float(A=0, B=1, Alpha=$User.CharacterPart)
    Use Spawn Probability=true
    UseLoopCountLimit=true
   )
  ]
  particleSpawnScript [
   &PartitionParticles(Particles Per Partition=1, Partition Mode:ENiagara_PartitionMode=::NewEnumerator1)
   &InitializeParticle(Color Mode=::NewEnumerator0, UsePositionOffset=false, Write Material Random=true)
   /SetVariables() {isSetParameters=true}
  ]
  particleUpdateScript [
   &ParticleState(Kill Particles When Lifetime Has Elapsed=false, Let Infinitely Lived Particles Die When Emitter Deactivates=true)
   /SetVariables(Particles.MainColor:LinearColor=&MultiplyLinearColor(Linear Color=<rgba 1 1 1 1>, Scale Factor=<1 1 1 1>)) {isSetParameters=true}
   /SetVariables(
    Particles.SoftGlowCameraOffset=2
    Particles.SoftGlowColor:LinearColor=&MultiplyLinearColor(
     Linear Color=&VariantColorSelector(Variant Amount=0, Original Color=<rgba 0.4624 0.8 0.4192 1>, Variant 1 Color=<rgba 0.5 0.1595 0.5 1>, Variant 2 Color=?)
     Scale Factor=&MakeVector4_FromFloat(Float=&Multiply_Float(
       A=$System.ReactivityTiming
       B=&Lerp_Float(
        A=$System.EyeBlinkL
        B=$System.EyeBlinkR
        Alpha=&MakeFloatFromInt(Int=$Particles.PartitionParticles.Partition)
       )
      ))
    )
    Particles.SoftGlowSize:Vector2f=&Multiply_Vector2D_ByFloat(
     Float=&MakeCustomFloatFromBool(Bool=$User.bisHeadHidden, True Float=0, False Float=1)
     Vector2D=&Vector2DFromFloat(Value=4)
    )
   ) {isSetParameters=true}
   /SetVariables(
    Particles.FakeLightColor:LinearColor=&MultiplyLinearColor(
     Linear Color=&VariantColorSelector(
      Variant Amount=0
      Morning Color=<rgba 0.8196 0.4824 1 1>
      Day Color=<rgba 1 1 1 1>
      Evening Color=<rgba 0.974 0.4513 0.2549 1>
      Night Color=<rgba 0.2 0.251 0.502 1>
      Use TOD?=true
      Original Color=<rgba 0.1156 0.2 0.1048 1>
      Variant 1 Color=<rgba 0.5 0.1595 0.5 1>
      Variant 2 Color=?
     )
     Scale Factor=&MakeVector4_FromFloat(Float=&Multiply_Float(
       A=$System.ReactivityTiming
       B=&Lerp_Float(
        A=$System.EyeBlinkL
        B=$System.EyeBlinkR
        Alpha=&MakeFloatFromInt(Int=$Particles.PartitionParticles.Partition)
       )
      ))
    )
    Particles.FakeLightSize:Vector2f=&Multiply_Vector2D_ByFloat(
     Float=&MakeCustomFloatFromBool(Bool=$User.bisHeadHidden, True Float=0, False Float=1)
     Vector2D=&Vector2DFromFloat(Value=10)
    )
   ) {isSetParameters=true}
   /SetVariables(
    Particles.CapsuleGlowColor:LinearColor=&MultiplyLinearColor(
     Linear Color=&VariantColorSelector(Variant Amount=0, Original Color=<rgba 0.1156 0.2 0.1048 1>, Variant 1 Color=<rgba 0.5 0.1595 0.5 1>, Variant 2 Color=?)
     Scale Factor=&MakeVector4_FromFloat(Float=&Multiply_Float(
       A=&Lerp_Float(
        A=$System.EyeBlinkL
        B=$System.EyeBlinkR
        Alpha=&MakeFloatFromInt(Int=$Particles.PartitionParticles.Partition)
       )
       B=$System.ReactivityTiming
      ))
    )
    Particles.CapsuleSpriteSize:Vector2f=&Multiply_Vector2D_ByFloat(
     Float=&MakeCustomFloatFromBool(Bool=$User.bisHeadHidden, True Float=0, False Float=1)
     Vector2D=&Vector2DFromFloat(Value=5)
    )
   ) {isSetParameters=true}
   &SkeletalMeshLocation(
    Bone / Socket Index=$Particles.PartitionParticles.Partition
    MaskSpawnGroup=false
    Mesh Sampling Type=::NewEnumerator5
    Skeletal Mesh=$System.Skeletal Mesh Head
    Socket Sampling Mode=::NewEnumerator5
   ) {module=SkeletalMeshLocation001}
   /SetVariables(
    Particles.LightColor=&MultiplyLinearColor(
     Linear Color=&VariantColorSelector(Variant Amount=0, Original Color=<rgba 0.2312 0.4 0.2096 1>, Variant 1 Color=<rgba 0.5 0.1595 0.5 1>, Variant 2 Color=?)
     Scale Factor=&MakeFloat4FromVectorAndFloat(
      "Vector (RGB)"=&VectorFromFloat(Value=100)
      "Float (Alpha)"=&Multiply_Float(
       A=$System.ReactivityTiming
       B=&Lerp_Float(
        A=$System.EyeBlinkL
        B=$System.EyeBlinkR
        Alpha=&MakeFloatFromInt(Int=$Particles.PartitionParticles.Partition)
       )
      )
     )
    )
    Particles.LightPosition:NiagaraPosition=&AddVectorToPosition(
     Position=$Particles.SkeletalMeshLocation001.SampledPosition
     Vector=&RotateVectorByQuaternion(Vector To Rotate=<0 2 0>, Quaternion=$Particles.SkeletalMeshLocation001.SampledRotation)
    )
    Particles.LightRadius=3
   ) {isSetParameters=true}
   /SetVariables(Particles.CapsuleSpriteAlignment:Vector3f=&RotateVectorByQuaternion(Vector To Rotate=<1 0 0>, Quaternion=$Particles.SkeletalMeshLocation001.SampledRotation)) {isSetParameters=true}
   /SetVariables(Particles.CapsuleSpriteDynamic:Vector4f=&MakeFloat4FromVectorAndFloat(
     "Vector (RGB)"=&NormalizeVector(Vector To Normalize=$Particles.CapsuleSpriteAlignment, Safe Normalize=true, Safe Fallback Threshold=0, Safe Fallback Vector=<0 0 0.001>)
     "Float (Alpha)"=10
    )) {isSetParameters=true}
  ]
  rendererClasses [&NiagaraSpriteRendererProperties, &NiagaraLightRendererProperties]
  renderers [
   {rendererClass=&NiagaraSpriteRendererProperties}
   {rendererClass=&NiagaraSpriteRendererProperties, rendererIndex=1}
   {rendererClass=&NiagaraSpriteRendererProperties, rendererIndex=2}
   {rendererClass=&NiagaraLightRendererProperties, rendererIndex=3}
  ]
 }
}
summary {
 systemName=NS_Character_SnapFreeze_Graze
 emitters [
  {emitterName=BodyGlow, rendererClasses [&NiagaraSpriteRendererProperties]}
  {emitterName=Head, rendererClasses [&NiagaraSpriteRendererProperties, &NiagaraLightRendererProperties]}
  {emitterName=BraceletGlow, rendererClasses [&NiagaraSpriteRendererProperties]}
 ]
 userVariables [
  User.bIsEditorPreview {type=NiagaraBool, defaultValue=true}
  User.bIsFrontEnd {type=NiagaraBool}
  User.bIsFrontEndPreview {type=NiagaraBool}
  User.bisHandHidden {type=NiagaraBool}
  User.bisHeadHidden {type=NiagaraBool}
  User.CharacterPart {type=NiagaraFloat, defaultValue=0, description=""}
  User.GameTime {type=NiagaraFloat, defaultValue=0, description=""}
  User.Variant {type=NiagaraFloat, defaultValue=0, description=""}
 ]
}
