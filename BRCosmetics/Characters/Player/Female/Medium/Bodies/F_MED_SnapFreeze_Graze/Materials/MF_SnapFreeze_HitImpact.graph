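# Material function for the SnapFreeze Graze character body that draws a fading hit-impact reaction around a position.
# Normalized time since TimeStamp expands a ring, a scalar SphereMask on offset-perturbed distance with a sharp front and soft tail, plus a sphere widened by AccumulatedExcitement.
# Distance defaults to vertex-interpolated pre-skinning local position.
# It outputs a faded ring/sphere blend, plus separate Ring, Sphere and Fade values.
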MaterialFunction /BRCosmetics/Characters/Player/Female/Medium/Bodies/F_MED_SnapFreeze_Graze/Materials/MF_SnapFreeze_HitImpact
graph {nodes [
  n1 = Time()
  Fmod(A=@n1, B=Constant() {r=5})
  n2 = FunctionInput() {inputName=TimeStamp, inputType=FunctionInput_Scalar}
  n3 = Saturate(None=Divide(A=Subtract(A=@n1, B=@n2), B=ScalarParameter() {defaultValue=4, group=SnapFreeze_Reactivity, parameterName=HitFadeDuration}))
  n4 = Multiply(
   A=ScalarParameter() {defaultValue=25, group=SnapFreeze_Reactivity, parameterName=HitRingRadius}
   B=Power(Base=@n3, Exp=ScalarParameter() {defaultValue=0.25, group=SnapFreeze_Reactivity, parameterName=HitRingExpandExp}) {constExponent=0.15}
  )
  n5 = FunctionInput() {inputName=Position}
  n6 = FunctionInput(Preview=VertexInterpolator(VS=LocalPosition() {localOrigin=InstancePreSkinning}.XYZ).PS) {bUsePreviewValueAsDefault=true, inputName=LocalPosition}
  n7 = Add(
   A=Distance(A=@n5, B=@n6)
   B=Multiply(
    A=FunctionInput() {bUsePreviewValueAsDefault=true, inputName=Offset, inputType=FunctionInput_Scalar}
    B=ScalarParameter() {defaultValue=5, group=SnapFreeze_Reactivity, parameterName=HitPosOffsetStrength}
   )
  )
  n8 = SphereMask(A=@n4, B=@n7, Radius=If(A=@n4, B=@n7, "A > B"=Constant() {r=30}, "A < B"=Constant() {r=2})) {attenuationRadius=8, hardnessPercent=0}
  n9 = Power(Base=@n8) {constExponent=3}
  n10 = ScalarParameter() {defaultValue=0.1, group=SnapFreeze_Reactivity, parameterName=HitSphereExpandExp}
  n11 = ScalarParameter() {defaultValue=15, group=SnapFreeze_Reactivity, parameterName=HitSphereRadius}
  n12 = ScalarParameter() {defaultValue=50, group=SnapFreeze_Reactivity, parameterName=HitSphereRadius_Excited}
  n13 = FunctionInput() {bUsePreviewValueAsDefault=true, inputName=AccumulatedExcitement, inputType=FunctionInput_Scalar, sortPriority=10}
  n14 = SphereMask(A=@n5, B=@n6, Radius=Multiply(A=Power(Base=@n3, Exp=@n10) {constExponent=0.1}, B=LinearInterpolate(A=@n11, B=@n12, Alpha=@n13))) {attenuationRadius=40, hardnessPercent=0}
  n15 = ScalarParameter() {defaultValue=0.5, group=SnapFreeze_Reactivity, parameterName=HitReactBlendRingVsSphere, sliderMax=1}
  n16 = FunctionOutput(None=OneMinus(None=@n3)) {outputName=Fade}
  FunctionOutput(None=Multiply(A=LinearInterpolate(A=@n9, B=@n14, Alpha=@n15), B=Power(Base=@n16)))
  FunctionOutput(None=@n8) {outputName=Ring}
  FunctionOutput(None=@n14) {outputName=Sphere}
  Multiply(A=LinearInterpolate(Alpha=@n3) {constA=2, constB=20}) {constB=0.2}
  Power(Base=@n14) {constExponent=3}
 ]}
