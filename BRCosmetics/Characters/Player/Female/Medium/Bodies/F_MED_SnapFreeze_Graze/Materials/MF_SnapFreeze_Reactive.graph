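# Material function for the SnapFreeze Graze character skin that adds reactive glow to incoming material attributes.
# A mask texture drives fresnel-tinted base color and emissive details, modulated by an idle hex-pattern pulse from two phase-offset, noise-distorted sphere-mask waves in pre-skinned space.
# Timestamped hit vectors feed MF_SnapFreeze_HitImpact for patterned emissive impacts; slots 4–7 are left out of the combined hit mask.
# Static switches disable idle or impact effects.
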MaterialFunction /BRCosmetics/Characters/Player/Female/Medium/Bodies/F_MED_SnapFreeze_Graze/Materials/MF_SnapFreeze_Reactive
graph {nodes [
  n1 = MaterialFunctionCall!(
   TimeStamp=ScalarParameter() {group=SnapFreeze_Reactivity, parameterName=HitTimestamp4}
   LocalPosition=NamedRerouteUsage()
   Offset=NamedRerouteUsage()
   Position=VectorParameter() {group=SnapFreeze_Reactivity, parameterName=HitVector4}.RGB
   AccumulatedExcitement=NamedRerouteUsage()
  ) {materialFunction=&/MF_SnapFreeze_HitImpact}
  n2 = MaterialFunctionCall(
   TimeStamp=ScalarParameter() {group=SnapFreeze_Reactivity, parameterName=HitTimestamp5}
   LocalPosition=NamedRerouteUsage()
   Offset=NamedRerouteUsage()
   Position=VectorParameter() {group=SnapFreeze_Reactivity, parameterName=HitVector5}.RGB
   AccumulatedExcitement=NamedRerouteUsage()
  ) {materialFunction=&MF_SnapFreeze_HitImpact}
  n3 = MaterialFunctionCall(
   TimeStamp=ScalarParameter() {group=SnapFreeze_Reactivity, parameterName=HitTimestamp6}
   LocalPosition=NamedRerouteUsage()
   Offset=NamedRerouteUsage()
   Position=VectorParameter() {group=SnapFreeze_Reactivity, parameterName=HitVector6}.RGB
   AccumulatedExcitement=NamedRerouteUsage()
  ) {materialFunction=&MF_SnapFreeze_HitImpact}
  n4 = MaterialFunctionCall(
   TimeStamp=ScalarParameter() {group=SnapFreeze_Reactivity, parameterName=HitTimestamp7}
   LocalPosition=NamedRerouteUsage()
   Offset=NamedRerouteUsage()
   Position=VectorParameter() {group=SnapFreeze_Reactivity, parameterName=HitVector7}.RGB
   AccumulatedExcitement=NamedRerouteUsage()
  ) {materialFunction=&MF_SnapFreeze_HitImpact}
  Add(A=Add(A=@n1.Result, B=@n2.Result), B=Add(A=@n3.Result, B=@n4.Result))
  n5 = Add(A=@n1.Fade, B=@n2.Fade)
  Add(A=@n5, B=Add(A=@n3.Fade, B=@n4.Fade))
  Add()
  n6 = LinearInterpolate(
   A=ScalarParameter() {defaultValue=0.1, group=SnapFreeze_Reactivity, parameterName=HitReactBrightness_OffPattern}
   B=ScalarParameter() {defaultValue=50, group=SnapFreeze_Reactivity, parameterName=HitReactBrightness}
   Alpha=NamedRerouteUsage()
  ) {constA=0.01}
  Divide(A=@n6, B=Max(A=NamedRerouteUsage()))
  n7 = GetMaterialAttributes(None=FunctionInput() {inputName=Material, inputType=FunctionInput_MaterialAttributes}) {attributeGetTypes [#1, #2]}
  n8 = VectorParameter() {group=SnapFreeze_Reactivity, parameterName=GlowingDetails_BaseColor, sortPriority=50, defaultValue=<rgba 0.2044 0 0.276 1>}
  n9 = VectorParameter() {group=SnapFreeze_Reactivity, parameterName=GlowingDetails_BaseColor_Rim, sortPriority=50, defaultValue=<rgba 0.2044 0 0.276 1>}
  n10 = Power(Base=NamedRerouteUsage(), Exp=ScalarParameter() {defaultValue=0.2, group=SnapFreeze_Reactivity, parameterName=GlowingDetails_FresnelExp}) {constExponent=0.2}
  n11 = LinearInterpolate(A=@n8.RGB, B=@n9.RGB, Alpha=@n10)
  n12 = TextureSampleParameter2D() {
   group=SnapFreeze_Reactivity
   parameterName=ReactiveMask
   samplerSource=SSM_Wrap_WorldGroupSettings
   samplerType=SAMPLERTYPE_Masks
   sortPriority=0
   texture=/BRCosmetics/Characters/Player/Female/Medium/Bodies/F_MED_SnapFreeze_Graze/Textures/T_SnapFreeze_Graze_Body_GM
  }
  n13 = MaterialFunctionCall(Float3=FunctionInput(Preview=@n12.RGB) {inputName=Masks}) {materialFunction=&BreakOutFloat3Components}
  n14 = NamedRerouteDeclaration(None=@n13.R) {name=GM_Red, nodeColor=<rgba 0.0078 1 0 1>}
  n15 = LinearInterpolate(A=@n7.BaseColor, B=@n11, Alpha=@n14)
  n16 = Add(A=@n7.BaseColor, B=Multiply(A=@n14, B=@n11))
  n17 = VectorParameter() {group=SnapFreeze_Reactivity, parameterName=GlowingDetails_EmissiveColor, sortPriority=50, defaultValue=<rgba 0.0942 0 0.709 1>}
  n18 = VectorParameter() {group=SnapFreeze_Reactivity, parameterName=GlowingDetails_EmissiveColor_Rim, sortPriority=50, defaultValue=<rgba 0.0942 0 0.709 1>}
  n19 = Multiply(
   A=Multiply(
    A=NamedRerouteDeclaration(None=@n13.B) {name=GM_Blue, nodeColor=<rgba 0.0078 1 0 1>}
    B=LinearInterpolate(A=@n17.RGB, B=@n18.RGB, Alpha=@n10)
   )
   B=ScalarParameter() {defaultValue=1, group=SnapFreeze_Reactivity, parameterName=GlowingDetails_Intensity, sortPriority=51}
  )
  n20 = GetMaterialAttributes(None=SetMaterialAttributes!(
    MaterialAttributes=@n7.MaterialAttributes
    BaseColor=StaticSwitchParameter(True=@n15, False=@n16) {group=SnapFreeze_Reactivity, parameterName=GlowingDetails_ReplaceBaseColor}
    EmissiveColor=Add(A=@n7.EmissiveColor, B=@n19)
   ) {attributeSetTypes [#1, #2]}) {attributeGetTypes [#2, #3, #1]}
  n21 = NamedRerouteUsage()
  n22 = Time()
  n23 = ScalarParameter() {defaultValue=0.15, group=SnapFreeze_Reactivity, parameterName=SFG_PulseSpeed}
  n24 = Frac(None=Multiply(A=@n22, B=@n23))
  n25 = OneMinus(None=@n24)
  n26 = ScalarParameter() {defaultValue=0.1, group=SnapFreeze_Reactivity, parameterName=SFG_IdlePulseWidth}
  n27 = ComponentMask(None=NamedRerouteUsage()) {b=true, r=true}
  n28 = TextureSample(UVs=Panner(Coordinate=Divide(A=@n27) {constB=64}) {speedX=-0.021, speedY=-0.027}, Tex=NamedRerouteUsage()) {
   samplerSource=SSM_Wrap_WorldGroupSettings
   samplerType=SAMPLERTYPE_LinearColor
   texture=/BRCosmetics/Characters/Player/Female/Medium/Bodies/F_MED_SnapFreeze_Graze/Textures/T_SnapFreeze_Graze_Pattern
  }
  n29 = Multiply(A=NamedRerouteDeclaration(None=@n28.R) {name=PS-Noise1, nodeColor=<rgba 1 0.2812 0 1>}) {constB=0.5}
  n30 = TextureObject() {samplerType=SAMPLERTYPE_Masks, texture=/Game/Global/Textures/Default/Noise/T_Noise_Cloud_Mask}
  n31 = LocalPosition() {localOrigin=InstancePreSkinning}
  n32 = FunctionInput(Preview=ShaderStageSwitch(PixelShader=VertexInterpolator(VS=@n31.XYZ).PS, VertexShader=@n31.XYZ)) {bUsePreviewValueAsDefault=true, inputName=PreskinnedPosition}
  n33 = Power(
   Base=MaterialFunctionCall(
    TextureObject=NamedRerouteDeclaration(None=@n30) {name=NoiseTex, nodeColor=<rgba 0.1484 0 1 1>}
    TextureSize=VectorParameter() {group=SnapFreeze_Reactivity, parameterName=SFG_PulseNoiseScale, defaultValue=<rgba 512 512 512 1>}.RGB
    WorldPosition=NamedRerouteDeclaration(None=@n32) {name=PS-LocalPos, nodeColor=<rgba 0 0.1562 1 1>}
   ) {materialFunction=&WorldAlignedTexture}.XYZ Texture
   Exp=ScalarParameter() {defaultValue=3, group=SnapFreeze_Reactivity, parameterName=SFG_WorldAlignedNoisePower}
  ) {constExponent=3}
  n34 = Subtract(A=Multiply(A=Add(A=@n29, B=@n33)) {constB=2})
  n35 = NamedRerouteUsage()
  n36 = Normalize(VectorInput=@n35)
  n37 = TextureSampleParameter2D(UVs=Multiply(A=TextureCoordinate(), B=ScalarParameter() {defaultValue=24, group=SnapFreeze_Reactivity, parameterName=SFG_Tiling})) {
   group=SnapFreeze_Reactivity
   parameterName=IdlePattern
   samplerType=SAMPLERTYPE_Masks
   texture=/BRCosmetics/Characters/Player/Male/Medium/Bodies/M_MED_Plum_Retro/Skin/Plum_Retro_Reactive/Textures/T_PlumHexPattern
  }
  n38 = Multiply(
   A=Subtract(A=NamedRerouteDeclaration(None=@n37.A) {name=Random, nodeColor=<rgba 0.6641 1 0 1>}) {constB=0.5}
   B=ScalarParameter() {defaultValue=15, group=SnapFreeze_Reactivity, parameterName=IdlePatternOffset}
  )
  n39 = SphereMask(
   A=Add(
    A=Multiply(A=@n34) {constB=64}
    B=VectorParameter() {group=SnapFreeze_Reactivity, parameterName=SFG_MaskPosition, defaultValue=<rgba 0 8 128 1>}.RGB
   )
   B=Add(A=@n35, B=Multiply(A=@n36, B=@n38))
   Radius=ScalarParameter() {defaultValue=192, group=SnapFreeze_Reactivity, parameterName=SFG_HitRadius}
   Hardness=ScalarParameter() {group=SnapFreeze_Reactivity, parameterName=SFG_MaskHardness}
  )
  n40 = Saturate(None=Multiply(A=@n25) {constB=8})
  n41 = Saturate(None=Multiply(A=@n24) {constB=8})
  n42 = Frac(None=Multiply(A=Add(A=@n22) {constB=2.355}, B=@n23))
  n43 = OneMinus(None=@n42)
  n44 = Saturate(None=Multiply(A=@n43) {constB=8})
  n45 = Saturate(None=Multiply(A=@n42) {constB=8})
  n46 = Max(
   A=Multiply(A=SphereMask(A=@n39, B=@n25, Radius=@n26) {hardnessPercent=0}, B=Min(A=@n40, B=@n41))
   B=Multiply(A=SphereMask(A=@n39, B=@n43, Radius=@n26) {hardnessPercent=0}, B=Min(A=@n44, B=@n45))
  )
  n47 = Clamp(None=@n46, Min=ScalarParameter() {defaultValue=0.25, group=SnapFreeze_Reactivity, parameterName=SFG_BaseEmissiveStrength})
  n48 = StaticSwitchParameter(True=@n20.EmissiveColor, False=Add(A=Multiply(A=@n20.EmissiveColor, B=@n47), B=@n21)) {parameterName=Disable Idle}
  n49 = NamedRerouteUsage()
  n50 = SetMaterialAttributes!(
   MaterialAttributes=@n20.MaterialAttributes
   EmissiveColor=StaticSwitchParameter(True=@n48, False=Add(A=@n48, B=@n49)) {parameterName=Disable Impacts}
   BaseColor
   Metallic
   Specular
   Roughness
  ) {attributeSetTypes [
    #2
    #1
    #4
    #5
    #6
   ]}
  FunctionOutput(None=@n50)
  LinearInterpolate(A=ScalarParameter() {defaultValue=0.2, group=SnapFreeze_Reactivity, parameterName=SFG_PatternLowInfluence}, Alpha=@n37.R) {constA=0.1}
  n51 = MaterialFunctionCall(
   TimeStamp=ScalarParameter() {group=SnapFreeze_Reactivity, parameterName=HitTimestamp0}
   LocalPosition=NamedRerouteUsage()
   Offset=NamedRerouteUsage()
   Position=VectorParameter() {parameterName=HitVector0}.RGB
   AccumulatedExcitement=NamedRerouteUsage()
  ) {materialFunction=&MF_SnapFreeze_HitImpact}
  n52 = MaterialFunctionCall(
   TimeStamp=ScalarParameter() {group=SnapFreeze_Reactivity, parameterName=HitTimestamp1}
   LocalPosition=NamedRerouteUsage()
   Offset=NamedRerouteUsage()
   Position=VectorParameter() {group=SnapFreeze_Reactivity, parameterName=HitVector1}.RGB
   AccumulatedExcitement=NamedRerouteUsage()
  ) {materialFunction=&MF_SnapFreeze_HitImpact}
  n53 = MaterialFunctionCall(
   TimeStamp=ScalarParameter() {group=SnapFreeze_Reactivity, parameterName=HitTimestamp2}
   LocalPosition=NamedRerouteUsage()
   Offset=NamedRerouteUsage()
   Position=VectorParameter() {group=SnapFreeze_Reactivity, parameterName=HitVector2}.RGB
   AccumulatedExcitement=NamedRerouteUsage()
  ) {materialFunction=&MF_SnapFreeze_HitImpact}
  n54 = MaterialFunctionCall(
   TimeStamp=ScalarParameter() {group=SnapFreeze_Reactivity, parameterName=HitTimestamp3}
   LocalPosition=NamedRerouteUsage()
   Offset=NamedRerouteUsage()
   Position=VectorParameter() {group=SnapFreeze_Reactivity, parameterName=HitVector3}.RGB
   AccumulatedExcitement=NamedRerouteUsage()
  ) {materialFunction=&MF_SnapFreeze_HitImpact}
  n55 = Add(A=Add(A=@n51.Fade, B=@n52.Fade), B=Add(A=@n53.Fade, B=@n54.Fade))
  MaterialFunctionCall(Number=Add(A=@n5, B=@n55), MaximumNumberOfDigits=Constant() {r=5}, UVs=Frac(None=TextureCoordinate() {uTiling=8, vTiling=8})) {materialFunction=&DebugScalarValues}
  n56 = Power() {constExponent=3}
  Multiply(A=SmoothStep(Value=Add(A=@n56)), B=Power())
  Multiply(A=@n21, B=NamedRerouteUsage())
  NamedRerouteDeclaration(None=SmoothStep(Value=@n46)) {name=PulseNoise, nodeColor=<rgba 1 0.0234 0 1>}
  n57 = Fresnel() {baseReflectFraction=0, exponent=1}
  n58 = Power(Base=@n57, Exp=ScalarParameter() {defaultValue=3, group=SnapFreeze_Reactivity, parameterName=SFG_FresnelPower})
  n59 = NamedRerouteUsage()
  n60 = Multiply(
   A=LinearInterpolate(
    B=ScalarParameter() {defaultValue=3, group=SnapFreeze_Reactivity, parameterName=SFG_IdleFresnelBrightness}
    Alpha=NamedRerouteDeclaration(None=@n58) {name=Fresnel, nodeColor=<rgba 0.5703 1 0 1>}
   ) {constA=1, constB=3}
   B=Multiply(A=@n37.R, B=@n59)
  )
  n61 = VectorParameter() {group=SnapFreeze_Reactivity, parameterName=IdlePatternColorLow, defaultValue=<rgba 0.0009 0 0.0521 1>}
  n62 = VectorParameter() {group=SnapFreeze_Reactivity, parameterName=IdlePatternColorHigh, defaultValue=<rgba 0.6788 0.0365 1 1>}
  n63 = Multiply(A=NamedRerouteUsage(), B=Multiply(A=@n60, B=LinearInterpolate(A=@n61.RGB, B=@n62.RGB, Alpha=@n60)))
  n64 = NamedRerouteUsage()
  NamedRerouteDeclaration(None=Multiply(A=Add(A=@n63, B=@n64), B=ScalarParameter() {defaultValue=3, group=SnapFreeze_Reactivity, parameterName=SFG_PatternBrightness}) {constB=7}) {name=PatternColor, nodeColor=<rgba 0 0.1797 1 1>}
  NamedRerouteDeclaration(None=@n55) {name=TotalExcitement, nodeColor=<rgba 1 0.0234 0 1>}
  NamedRerouteDeclaration(None=Max(A=Max(A=@n51.Result, B=@n52.Result), B=Max(A=@n53.Result, B=@n54.Result))) {name=HitMask, nodeColor=<rgba 1 0.1641 0 1>}
  n65 = TextureSampleParameter2D(UVs=Multiply(A=TextureCoordinate(), B=ScalarParameter() {defaultValue=12, parameterName=HitReactPatternScale})) {
   group=SnapFreeze_Reactivity
   parameterName=HitReactPattern
   samplerSource=SSM_Wrap_WorldGroupSettings
   samplerType=SAMPLERTYPE_Masks
   texture=&T_PlumHexPattern
  }
  n66 = Add(A=Multiply(A=NamedRerouteUsage(), B=ScalarParameter() {defaultValue=0.15, parameterName=HitReactAdditiveScale}), B=NamedRerouteUsage())
  n67 = LinearInterpolate(
   A=@n66
   B=Multiply(A=@n66, B=@n65.R)
   Alpha=ScalarParameter() {defaultValue=1, group=SnapFreeze_Reactivity, parameterName=HitReactPatternStrength}
  )
  n68 = Power(Base=@n67) {constExponent=2.5}
  n69 = Multiply(
   A=LinearInterpolate(
    A=VectorParameter() {group=SnapFreeze_Reactivity, parameterName=HitReactColorLow, defaultValue=<rgba 0.0009 0 0.0521 1>}.RGB
    B=VectorParameter() {group=SnapFreeze_Reactivity, parameterName=HitReactColorHigh, defaultValue=<rgba 0.6788 0.0365 1 1>}.RGB
    Alpha=Power(Base=@n67)
   )
   B=@n6
  )
  n70 = ScalarParameter() {defaultValue=3, group=SnapFreeze_Reactivity, parameterName=HitReactFresnelExp}
  n71 = ScalarParameter() {defaultValue=1, group=SnapFreeze_Reactivity, parameterName=HitReactFresnelBase}
  NamedRerouteDeclaration(None=Multiply(A=Multiply(A=@n68, B=@n69), B=Fresnel(ExponentIn=@n70, BaseReflectFractionIn=@n71))) {name=ImpactFX, nodeColor=<rgba 0.7109 0 1 1>}
  NamedRerouteDeclaration(None=Add(A=Add(A=@n51.Result, B=@n52.Result), B=Add(A=@n53.Result, B=@n54.Result))) {name=HitMaskAdditive, nodeColor=<rgba 0.4766 0 1 1>}
  NamedRerouteDeclaration(None=VertexInterpolator(VS=LocalPosition() {localOrigin=InstancePreSkinning}.XYZ).PS) {name=PreskinnedPos, nodeColor=<rgba 0 1 0.0859 1>}
  n72 = Multiply(
   A=ScalarParameter() {defaultValue=0.75, group=SnapFreeze_Reactivity, parameterName=SFG_PatternFresnelStrength}
   B=Saturate(None=NamedRerouteUsage())
  )
  n73 = ScalarParameter() {defaultValue=2, parameterName=SFG_PatternFresnelExponent}
  n74 = ScalarParameter() {parameterName=SFG_PatternFresnelBase}
  NamedRerouteDeclaration(None=Multiply(A=Multiply(A=@n61.RGB, B=@n72), B=Fresnel(ExponentIn=@n73, BaseReflectFractionIn=@n74))) {name=PatternFresnel, nodeColor=<rgba 0 1 0.7891 1>}
  NamedRerouteDeclaration(None=@n67) {name=Debug, nodeColor=<rgba 0.3125 1 0 1>}
  NamedRerouteDeclaration(None=@n65.A) {name=HitOffset, nodeColor=<rgba 0.4531 0 1 1>}
  NamedRerouteDeclaration(None=@n13.G) {name=GM_Green, nodeColor=<rgba 1 0 0.0937 1>}
  NamedRerouteUsage()
  OneMinus(None=@n57)
  Saturate(None=@n59)
 ]}
