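# Parent character material for the SnapFreeze Graze body, adding switchable reactive and elimination-glow effects between the shared 2023 character PreFX and PostFX functions.
# ReactiveMask RGB and pre-skinned position feed MF_SnapFreeze_Reactive for the reactive look.
# For kill reactivity, time since Elim_StartTime drives a fast cubic ramp-in and slower fade-out.
# That envelope scales a Fresnel-blended color and brightness, masked by ReactiveMask alpha and added to emissive.
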Material /BRCosmetics/Characters/Player/Female/Medium/Bodies/F_MED_SnapFreeze_Graze/Materials/M_SnapFreeze_PARENT
graph {
 properties {
  bHasPixelAnimation=true
  bUseMaterialAttributes=true
  bUsedWithMeshDeformer=true
  bUsedWithMorphTargets=true
  bUsedWithSkeletalMesh=true
  pixelDepthOffsetMode=PDOM_Legacy
  shadingModel=MSM_FromMaterialExpression
  usedShadingModels=Default Lit | Subsurface | Clear Coat | Subsurface Profile | Cloth
 }
 nodes [
  n1 = MaterialFunctionCall() {materialFunction=&MF_CharacterParent_2023_PreFX}
  n2 = TextureSampleParameter2D() {
   group=SnapFreeze_Reactivity
   parameterName=ReactiveMask
   samplerSource=SSM_Wrap_WorldGroupSettings
   samplerType=SAMPLERTYPE_Masks
   sortPriority=0
   texture=/BRCosmetics/Characters/Player/Female/Medium/Bodies/F_MED_SnapFreeze_Graze/Textures/T_SnapFreeze_Graze_Body_GM
  }
  n3 = StaticSwitchParameter(
   True=MaterialFunctionCall!(PreskinnedPosition=@n1.Pre skinned Local Position, Masks=@n2.RGB, Material=@n1.Pre skinned Local Position) {materialFunction=&/MF_SnapFreeze_Reactive}.Result
   False=@n1.Result
  ) {defaultValue=true, group=SnapFreeze_Reactivity, parameterName=UseReactiveLook, sortPriority=0}
  n4 = GetMaterialAttributes(None=@n3) {attributeGetTypes [#1]}
  n5 = Subtract(A=Time(), B=ScalarParameter() {group=SnapFreeze_Reactivity, parameterName=Elim_StartTime})
  n6 = Divide(A=@n5, B=ScalarParameter() {defaultValue=0.1, parameterName=Elim_RampUpDuration})
  n7 = Saturate(None=Divide(A=@n5, B=ScalarParameter() {defaultValue=3, parameterName=Elim_Duration}))
  n8 = Multiply(A=Power(Base=Saturate(None=@n6)) {constExponent=3}, B=Power(Base=OneMinus(None=@n7)) {constExponent=0.8})
  n9 = ScalarParameter() {defaultValue=0.1, parameterName=Elim_BrightnessFresnelExp}
  n10 = ScalarParameter() {defaultValue=0.2, parameterName=Elim_ColorFresnelExp}
  n11 = Multiply(
   A=LinearInterpolate(
    A=ScalarParameter() {defaultValue=20, parameterName=Elim_BrightnessLow}
    B=ScalarParameter() {defaultValue=1, parameterName=Elim_BrightnessHigh}
    Alpha=Power(Base=@n1.Fresnel, Exp=@n9)
   )
   B=LinearInterpolate(
    A=VectorParameter() {parameterName=Elim_ColorLow, defaultValue=<rgba 0.8368 0.5136 1 1>}.RGB
    B=VectorParameter() {parameterName=Elim_ColorHigh, defaultValue=<rgba 0.4305 0 1 1>}.RGB
    Alpha=Power(Base=@n1.Fresnel, Exp=@n10)
   )
  )
  n12 = Add(A=@n4.Emissive Color, B=Multiply(A=@n2.A, B=Multiply(A=@n8, B=@n11)))
  MaterialFunctionCall(
   Material=StaticSwitchParameter(
    True=SetMaterialAttributes(MaterialAttributes=@n4.MaterialAttributes, Emissive Color=@n12) {attributeSetTypes [#1]}
    False=@n3
   ) {defaultValue=true, group=SnapFreeze_Reactivity, parameterName=UseKillReactivity, sortPriority=0}
   M=@n1.M
   Pre Skinned Local Position=@n1.M
   Distance Blend=@n1.M
   Fresnel=@n1.M
  ) {materialFunction=&MF_CharacterParent_2023_PostFX}
 ]
}
