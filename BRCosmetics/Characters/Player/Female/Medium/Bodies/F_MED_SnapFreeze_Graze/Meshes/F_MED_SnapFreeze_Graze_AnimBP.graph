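# Post-process style AnimBlueprint for the SnapFreeze Graze female body, adding secondary motion to fur, necklace and bracelet.
# The parent mesh pose feeds a Control Rig, then weighted necklace constraints and per-bone AnimDynamics springs.
# Each update computes gravity override and front-end idle alphas, and drives material-based cloth through WPO functions and skydiving-dependent wind parameters.
# Dynamic body and overlay materials are cached when character customization loads.
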AnimBlueprint /BRCosmetics/Characters/Player/Female/Medium/Bodies/F_MED_SnapFreeze_Graze/Meshes/F_MED_SnapFreeze_Graze_AnimBP
target_skeleton=&/F_MED_SnapFreeze_Graze_Skeleton
graphs [
 AnimGraph {nodes [
   n1 = Control Rig(Source=Copy Pose From Mesh() {bUseAttachedParent=true}) {
    alphaBoolBlend {blendOption=Linear}
    controlRigAssetReference {blueprintRigClass=&/F_MED_SnapFreeze_Graze_CtrlRig.F_MED_SnapFreeze_Graze_CtrlRig_C}
   }
   n2 = "Constraint - dyn_necklace_bck_c (PPPPPP)"!(
    ComponentPose=Local To Component(LocalPose=@n1)
    Alpha=1
    ConstraintWeights_0=0.9082
    ConstraintWeights_1:float=0.0014
    ConstraintWeights_2:float=0.002
    ConstraintWeights_3:float=0.0283
    ConstraintWeights_4:float=0.0127
    ConstraintWeights_5:float=0.0475
   ) -> (Pose) {
    alphaBoolBlend {blendOption=Linear}
    boneToModify {boneName=dyn_necklace_bck_c}
    constraintSetup [
     {targetBone {boneName=spine_05}}
     {targetBone {boneName=clavicle_l}}
     {targetBone {boneName=clavicle_r}}
     {targetBone {boneName=neck_01}}
     {targetBone {boneName=head}}
     {targetBone {boneName=neck_02}}
    ]
    constraintWeights [1,1,1,1,1,1]
   }
   n3 = "Constraint - dyn_necklace_bck_l (PPPPPP)"!(
    ComponentPose=@n2
    Alpha=1
    ConstraintWeights_0=0.9185
    ConstraintWeights_1:float=0.0059
    ConstraintWeights_2:float=0.0009
    ConstraintWeights_3:float=0.0272
    ConstraintWeights_4:float=0.0059
    ConstraintWeights_5:float=0.0416
   ) -> (Pose) {
    alphaBoolBlend {blendOption=Linear}
    boneToModify {boneName=dyn_necklace_bck_l}
    constraintSetup [
     {targetBone {boneName=spine_05}}
     {targetBone {boneName=clavicle_l}}
     {targetBone {boneName=clavicle_r}}
     {targetBone {boneName=neck_01}}
     {targetBone {boneName=head}}
     {targetBone {boneName=neck_02}}
    ]
    constraintWeights [1,1,1,1,1,1]
   }
   n4 = "Constraint - dyn_necklace_bck_r (PPPPPP)"!(
    ComponentPose=@n3
    Alpha=1
    ConstraintWeights_0=0.8253
    ConstraintWeights_1:float=0.0034
    ConstraintWeights_2:float=0.0042
    ConstraintWeights_3:float=0.059
    ConstraintWeights_4:float=0.0184
    ConstraintWeights_5:float=0.0896
   ) -> (Pose) {
    alphaBoolBlend {blendOption=Linear}
    boneToModify {boneName=dyn_necklace_bck_r}
    constraintSetup [
     {targetBone {boneName=spine_05}}
     {targetBone {boneName=clavicle_l}}
     {targetBone {boneName=clavicle_r}}
     {targetBone {boneName=neck_01}}
     {targetBone {boneName=head}}
     {targetBone {boneName=neck_02}}
    ]
    constraintWeights [1,1,1,1,1,1]
   }
   n5 = "Constraint - dyn_necklace_side_l (PPPPP)"!(
    ComponentPose=@n4
    Alpha=1
    ConstraintWeights_0=0.0038
    ConstraintWeights_1:float=0.8381
    ConstraintWeights_2:float=0.1556
    ConstraintWeights_3:float=0.0004
    ConstraintWeights_4:float=0.0021
   ) -> (Pose) {
    alphaBoolBlend {blendOption=Linear}
    boneToModify {boneName=dyn_necklace_side_l}
    constraintSetup [
     {targetBone {boneName=spine_04}}
     {targetBone {boneName=spine_05}}
     {targetBone {boneName=clavicle_l}}
     {targetBone {boneName=neck_01}}
     {targetBone {boneName=neck_02}}
    ]
    constraintWeights [1,1,1,1,1]
   }
   n6 = "Constraint - dyn_necklace_side_r (PPPPP)"!(
    ComponentPose=@n5
    Alpha=1
    ConstraintWeights_0=0.0023
    ConstraintWeights_1:float=0.8394
    ConstraintWeights_2:float=0.1495
    ConstraintWeights_3:float=0.002
    ConstraintWeights_4:float=0.0068
   ) -> (Pose) {
    alphaBoolBlend {blendOption=Linear}
    boneToModify {boneName=dyn_necklace_side_r}
    constraintSetup [
     {targetBone {boneName=spine_04}}
     {targetBone {boneName=spine_05}}
     {targetBone {boneName=clavicle_r}}
     {targetBone {boneName=neck_01}}
     {targetBone {boneName=neck_02}}
    ]
    constraintWeights [1,1,1,1,1]
   }
   n7 = Anim Dynamics - Bone: dyn_fur_l_A!(ComponentPose=@n6, Alpha=1, GravityOverride:Vector="0,0,-350", ExternalForce="0,0,0") -> (Pose) {
    angularDampingOverride=0.9999
    angularSpringConstant=5000
    bAngularSpring=true
    bGravityOverrideInSimSpace=true
    bOverrideAngularDamping=true
    bUseGravityOverride=true
    linearDampingOverride=0.9
    linearSpringConstant=150
    numSolverIterationsPreUpdate=3
    simulationSpace=BoneRelative
    alphaBoolBlend {blendOption=Linear}
    boundBone {boneName=dyn_fur_l_A}
    componentAppliedLinearAccClamp=<10000 10000 10000>
    componentLinearVelScale=<0.1 0.1 0.1>
    gravityOverride {z=-350}
    physicsBodyDefinitions [{
      boundBone {boneName=dyn_fur_l_A}
      boxExtents=<20 20 20>
      constraintSetup {angularLimitsMax=<5 0 3>, angularLimitsMin=<-5 -30 -3>, angularTarget=<1 0 0>}
      localJointOffset=<2 0 0>
     }]
    relativeSpaceBone {boneName=root}
   }
   n8 = Anim Dynamics - Bone: dyn_fur_l_B!(ComponentPose=@n7, Alpha=1, GravityOverride:Vector="0,0,-350", ExternalForce="0,0,0") -> (Pose) {
    angularDampingOverride=0.9999
    angularSpringConstant=5000
    bAngularSpring=true
    bGravityOverrideInSimSpace=true
    bOverrideAngularDamping=true
    bUseGravityOverride=true
    linearDampingOverride=0.9
    linearSpringConstant=150
    numSolverIterationsPreUpdate=3
    simulationSpace=BoneRelative
    alphaBoolBlend {blendOption=Linear}
    boundBone {boneName=dyn_fur_l_B}
    componentAppliedLinearAccClamp=<10000 10000 10000>
    componentLinearVelScale=<0.1 0.1 0.1>
    gravityOverride {z=-350}
    physicsBodyDefinitions [{
      boundBone {boneName=dyn_fur_l_B}
      boxExtents=<23 23 23>
      constraintSetup {angularLimitsMax=<5 0 3>, angularLimitsMin=<-5 -30 -3>, angularTarget=<1 0 0>}
      localJointOffset=<2 0 0>
     }]
    relativeSpaceBone {boneName=root}
   }
   n9 = Anim Dynamics - Bone: dyn_fur_l_C!(ComponentPose=@n8, Alpha=1, GravityOverride:Vector="0,0,-350", ExternalForce="0,0,0") -> (Pose) {
    angularDampingOverride=0.9999
    angularSpringConstant=5000
    bAngularSpring=true
    bGravityOverrideInSimSpace=true
    bOverrideAngularDamping=true
    bUseGravityOverride=true
    linearDampingOverride=0.9
    linearSpringConstant=150
    numSolverIterationsPreUpdate=3
    simulationSpace=BoneRelative
    alphaBoolBlend {blendOption=Linear}
    boundBone {boneName=dyn_fur_l_C}
    componentAppliedLinearAccClamp=<10000 10000 10000>
    componentLinearVelScale=<0.1 0.1 0.1>
    gravityOverride {z=-350}
    physicsBodyDefinitions [{
      boundBone {boneName=dyn_fur_l_C}
      boxExtents=<23 23 23>
      constraintSetup {angularLimitsMax=<5 0 3>, angularLimitsMin=<-5 -30 -3>, angularTarget=<1 0 0>}
      localJointOffset=<2 0 0>
     }]
    relativeSpaceBone {boneName=root}
   }
   n10 = Anim Dynamics - Bone: dyn_fur_l_D!(ComponentPose=@n9, Alpha=1, GravityOverride:Vector="0,0,-350", ExternalForce="0,0,0") -> (Pose) {
    angularDampingOverride=0.9999
    angularSpringConstant=5000
    bAngularSpring=true
    bGravityOverrideInSimSpace=true
    bOverrideAngularDamping=true
    bUseGravityOverride=true
    linearDampingOverride=0.9
    linearSpringConstant=150
    numSolverIterationsPreUpdate=3
    simulationSpace=BoneRelative
    alphaBoolBlend {blendOption=Linear}
    boundBone {boneName=dyn_fur_l_D}
    componentAppliedLinearAccClamp=<10000 10000 10000>
    componentLinearVelScale=<0.1 0.1 0.1>
    gravityOverride {z=-350}
    physicsBodyDefinitions [{
      boundBone {boneName=dyn_fur_l_D}
      boxExtents=<20 20 20>
      constraintSetup {angularLimitsMax=<5 0 3>, angularLimitsMin=<-5 -30 -3>, angularTarget=<1 0 0>}
      localJointOffset=<2 0 0>
     }]
    relativeSpaceBone {boneName=root}
   }
   n11 = Anim Dynamics - Bone: dyn_fur_r_A!(ComponentPose=@n10, Alpha=1, GravityOverride:Vector="0,0,-350", ExternalForce="0,0,0") -> (Pose) {
    angularDampingOverride=0.9999
    angularSpringConstant=5000
    bAngularSpring=true
    bGravityOverrideInSimSpace=true
    bOverrideAngularDamping=true
    bUseGravityOverride=true
    linearDampingOverride=0.9
    linearSpringConstant=150
    numSolverIterationsPreUpdate=3
    simulationSpace=BoneRelative
    alphaBoolBlend {blendOption=Linear}
    boundBone {boneName=dyn_fur_r_A}
    componentAppliedLinearAccClamp=<10000 10000 10000>
    componentLinearVelScale=<0.1 0.1 0.1>
    gravityOverride {z=-350}
    physicsBodyDefinitions [{
      boundBone {boneName=dyn_fur_r_A}
      boxExtents=<20 20 20>
      constraintSetup {angularLimitsMax=<5 0 3>, angularLimitsMin=<-5 -30 -3>, angularTarget=<1 0 0>}
      localJointOffset=<2 0 0>
     }]
    relativeSpaceBone {boneName=root}
   }
   n12 = Anim Dynamics - Bone: dyn_fur_r_B!(ComponentPose=@n11, Alpha=1, GravityOverride:Vector="0,0,-350", ExternalForce="0,0,0") -> (Pose) {
    angularDampingOverride=0.9999
    angularSpringConstant=5000
    bAngularSpring=true
    bGravityOverrideInSimSpace=true
    bOverrideAngularDamping=true
    bUseGravityOverride=true
    linearDampingOverride=0.9
    linearSpringConstant=150
    numSolverIterationsPreUpdate=3
    simulationSpace=BoneRelative
    alphaBoolBlend {blendOption=Linear}
    boundBone {boneName=dyn_fur_r_B}
    componentAppliedLinearAccClamp=<10000 10000 10000>
    componentLinearVelScale=<0.1 0.1 0.1>
    gravityOverride {z=-350}
    physicsBodyDefinitions [{
      boundBone {boneName=dyn_fur_r_B}
      boxExtents=<23 23 23>
      constraintSetup {angularLimitsMax=<5 0 3>, angularLimitsMin=<-5 -30 -3>, angularTarget=<1 0 0>}
      localJointOffset=<2 0 0>
     }]
    relativeSpaceBone {boneName=root}
   }
   n13 = Anim Dynamics - Bone: dyn_fur_r_C!(ComponentPose=@n12, Alpha=1, GravityOverride:Vector="0,0,-350", ExternalForce="0,0,0") -> (Pose) {
    angularDampingOverride=0.9999
    angularSpringConstant=5000
    bAngularSpring=true
    bGravityOverrideInSimSpace=true
    bOverrideAngularDamping=true
    bUseGravityOverride=true
    linearDampingOverride=0.9
    linearSpringConstant=150
    numSolverIterationsPreUpdate=3
    simulationSpace=BoneRelative
    alphaBoolBlend {blendOption=Linear}
    boundBone {boneName=dyn_fur_r_C}
    componentAppliedLinearAccClamp=<10000 10000 10000>
    componentLinearVelScale=<0.1 0.1 0.1>
    gravityOverride {z=-350}
    physicsBodyDefinitions [{
      boundBone {boneName=dyn_fur_r_C}
      boxExtents=<23 23 23>
      constraintSetup {angularLimitsMax=<5 0 3>, angularLimitsMin=<-5 -30 -3>, angularTarget=<1 0 0>}
      localJointOffset=<2 0 0>
     }]
    relativeSpaceBone {boneName=root}
   }
   n14 = Anim Dynamics - Bone: dyn_fur_r_D!(ComponentPose=@n13, Alpha=1, GravityOverride:Vector="0,0,-350", ExternalForce="0,0,0") -> (Pose) {
    angularDampingOverride=0.9999
    angularSpringConstant=5000
    bAngularSpring=true
    bGravityOverrideInSimSpace=true
    bOverrideAngularDamping=true
    bUseGravityOverride=true
    linearDampingOverride=0.9
    linearSpringConstant=150
    numSolverIterationsPreUpdate=3
    simulationSpace=BoneRelative
    alphaBoolBlend {blendOption=Linear}
    boundBone {boneName=dyn_fur_r_D}
    componentAppliedLinearAccClamp=<10000 10000 10000>
    componentLinearVelScale=<0.1 0.1 0.1>
    gravityOverride {z=-350}
    physicsBodyDefinitions [{
      boundBone {boneName=dyn_fur_r_D}
      boxExtents=<20 20 20>
      constraintSetup {angularLimitsMax=<5 0 3>, angularLimitsMin=<-5 -30 -3>, angularTarget=<1 0 0>}
      localJointOffset=<2 0 0>
     }]
    relativeSpaceBone {boneName=root}
   }
   n15 = Anim Dynamics - Bone: dyn_necklace_bck_l(ComponentPose=@n14, Alpha=1, ExternalForce="0,0,0") {
    angularDampingOverride=0.89
    angularSpringConstant=100
    bAngularSpring=true
    bGravityOverrideInSimSpace=true
    bLinearSpring=true
    bOverrideAngularDamping=true
    bOverrideLinearDamping=true
    bUseGravityOverride=true
    linearDampingOverride=0.9999
    linearSpringConstant=600
    numSolverIterationsPreUpdate=3
    simulationSpace=BoneRelative
    alphaBoolBlend {blendOption=Linear}
    boundBone {boneName=dyn_necklace_bck_l}
    componentAppliedLinearAccClamp=<10000 10000 10000>
    componentLinearAccScale=<0.5 0.5 0.5>
    componentLinearVelScale=<0.5 0.5 0.5>
    gravityOverride {z=-980}
    physicsBodyDefinitions [{
      boundBone {boneName=dyn_necklace_bck_l}
      boxExtents=<150 150 150>
      constraintSetup {angularLimitsMax=<0 0 0>, angularLimitsMin=<0 0 0>, angularTarget=<1 0 0>, linearAxesMax=<0.5 0 1>}
      localJointOffset=<2 0 0>
     }]
    relativeSpaceBone {boneName=root}
   }
   n16 = Anim Dynamics - Bone: dyn_necklace_bck_c(ComponentPose=@n15, Alpha=1, ExternalForce="0,0,0") {
    angularDampingOverride=0.89
    angularSpringConstant=100
    bAngularSpring=true
    bGravityOverrideInSimSpace=true
    bLinearSpring=true
    bOverrideAngularDamping=true
    bOverrideLinearDamping=true
    bUseGravityOverride=true
    linearDampingOverride=0.9999
    linearSpringConstant=600
    numSolverIterationsPreUpdate=3
    simulationSpace=BoneRelative
    alphaBoolBlend {blendOption=Linear}
    boundBone {boneName=dyn_necklace_bck_c}
    componentAppliedLinearAccClamp=<10000 10000 10000>
    componentLinearAccScale=<0.5 0.5 0.5>
    componentLinearVelScale=<0.5 0.5 0.5>
    gravityOverride {z=-980}
    physicsBodyDefinitions [{
      boundBone {boneName=dyn_necklace_bck_c}
      boxExtents=<150 150 150>
      constraintSetup {angularLimitsMax=<0 0 0>, angularLimitsMin=<0 0 0>, angularTarget=<1 0 0>, linearAxesMax=<0.5 0 1>}
      localJointOffset=<2 0 0>
     }]
    relativeSpaceBone {boneName=root}
   }
   n17 = Anim Dynamics - Bone: dyn_necklace_bck_r(ComponentPose=@n16, Alpha=1, ExternalForce="0,0,0") {
    angularDampingOverride=0.89
    angularSpringConstant=100
    bAngularSpring=true
    bGravityOverrideInSimSpace=true
    bLinearSpring=true
    bOverrideAngularDamping=true
    bOverrideLinearDamping=true
    bUseGravityOverride=true
    linearDampingOverride=0.9999
    linearSpringConstant=600
    numSolverIterationsPreUpdate=3
    simulationSpace=BoneRelative
    alphaBoolBlend {blendOption=Linear}
    boundBone {boneName=dyn_necklace_bck_r}
    componentAppliedLinearAccClamp=<10000 10000 10000>
    componentLinearAccScale=<0.5 0.5 0.5>
    componentLinearVelScale=<0.5 0.5 0.5>
    gravityOverride {z=-980}
    physicsBodyDefinitions [{
      boundBone {boneName=dyn_necklace_bck_r}
      boxExtents=<150 150 150>
      constraintSetup {angularLimitsMax=<0 0 0>, angularLimitsMin=<0 0 0>, angularTarget=<1 0 0>, linearAxesMax=<0.5 0 1>}
      localJointOffset=<2 0 0>
     }]
    relativeSpaceBone {boneName=root}
   }
   n18 = Anim Dynamics - Bone: dyn_necklace_side_l(ComponentPose=@n17, Alpha=1, ExternalForce="0,0,0") {
    angularDampingOverride=0.89
    angularSpringConstant=100
    bAngularSpring=true
    bGravityOverrideInSimSpace=true
    bLinearSpring=true
    bOverrideAngularDamping=true
    bOverrideLinearDamping=true
    bUseGravityOverride=true
    linearDampingOverride=0.9999
    linearSpringConstant=600
    numSolverIterationsPreUpdate=3
    simulationSpace=BoneRelative
    alphaBoolBlend {blendOption=Linear}
    boundBone {boneName=dyn_necklace_side_l}
    componentAppliedLinearAccClamp=<10000 10000 10000>
    componentLinearAccScale=<0.5 0.5 0.5>
    componentLinearVelScale=<0.5 0.5 0.5>
    gravityOverride {z=-980}
    physicsBodyDefinitions [{
      boundBone {boneName=dyn_necklace_side_l}
      boxExtents=<150 150 150>
      constraintSetup {angularLimitsMax=<0 0 0>, angularLimitsMin=<0 0 0>, angularTarget=<1 0 0>, linearAxesMax=<0.5 0 1>}
      localJointOffset=<2 0 0>
     }]
    relativeSpaceBone {boneName=root}
   }
   n19 = Anim Dynamics - Bone: dyn_necklace_side_r(ComponentPose=@n18, Alpha=1, ExternalForce="0,0,0") {
    angularDampingOverride=0.89
    angularSpringConstant=100
    bAngularSpring=true
    bGravityOverrideInSimSpace=true
    bLinearSpring=true
    bOverrideAngularDamping=true
    bOverrideLinearDamping=true
    bUseGravityOverride=true
    linearDampingOverride=0.9999
    linearSpringConstant=600
    numSolverIterationsPreUpdate=3
    simulationSpace=BoneRelative
    alphaBoolBlend {blendOption=Linear}
    boundBone {boneName=dyn_necklace_side_r}
    componentAppliedLinearAccClamp=<10000 10000 10000>
    componentLinearAccScale=<0.5 0.5 0.5>
    componentLinearVelScale=<0.5 0.5 0.5>
    gravityOverride {z=-980}
    physicsBodyDefinitions [{
      boundBone {boneName=dyn_necklace_side_r}
      boxExtents=<150 150 150>
      constraintSetup {angularLimitsMax=<0 0 0>, angularLimitsMin=<0 0 0>, angularTarget=<1 0 0>, linearAxesMax=<0.5 0 1>}
      localJointOffset=<2 0 0>
     }]
    relativeSpaceBone {boneName=root}
   }
   n20 = Anim Dynamics - Bone: dyn_necklace_c_base(ComponentPose=@n19, Alpha=1, ExternalForce="0,0,0") {
    angularDampingOverride=0.89
    angularSpringConstant=100
    bAngularSpring=true
    bGravityOverrideInSimSpace=true
    bLinearSpring=true
    bOverrideAngularDamping=true
    bOverrideLinearDamping=true
    bUseGravityOverride=true
    linearDampingOverride=0.9999
    linearSpringConstant=600
    numSolverIterationsPreUpdate=3
    simulationSpace=BoneRelative
    alphaBoolBlend {blendOption=Linear}
    boundBone {boneName=dyn_necklace_c_base}
    componentAppliedLinearAccClamp=<10000 10000 10000>
    componentLinearAccScale=<0.5 0.5 0.5>
    componentLinearVelScale=<0.5 0.5 0.5>
    gravityOverride {z=-980}
    physicsBodyDefinitions [{
      boundBone {boneName=dyn_necklace_c_base}
      boxExtents=<150 150 150>
      constraintSetup {angularLimitsMax=<0 0 0>, angularLimitsMin=<0 0 0>, angularTarget=<1 0 0>, linearAxesMax=<0.5 0 1>}
      localJointOffset=<2 0 0>
     }]
    relativeSpaceBone {boneName=root}
   }
   n21 = "Constraint - dyn_necklace_l_02 (P)"(ComponentPose=@n20, Alpha=0.4, ConstraintWeights_0=1) {
    alphaBoolBlend {blendOption=Linear}
    boneToModify {boneName=dyn_necklace_l_02}
    constraintSetup [{targetBone {boneName=spine_04}}]
    constraintWeights [1]
   }
   n22 = "Constraint - dyn_necklace_r_02 (P)"(ComponentPose=@n21, Alpha=0.4, ConstraintWeights_0=1) {
    alphaBoolBlend {blendOption=Linear}
    boneToModify {boneName=dyn_necklace_r_02}
    constraintSetup [{targetBone {boneName=spine_04}}]
    constraintWeights [1]
   }
   n23 = Anim Dynamics - Bone: dyn_necklace_l_01!(ComponentPose=@n22, Alpha=1, GravityOverride:Vector="0,0,-350", ExternalForce="0,0,0") -> (Pose) {
    angularDampingOverride=0.89
    angularSpringConstant=100
    bAngularSpring=true
    bGravityOverrideInSimSpace=true
    bOverrideAngularDamping=true
    bUseGravityOverride=true
    linearDampingOverride=0.9
    linearSpringConstant=150
    numSolverIterationsPreUpdate=3
    simulationSpace=BoneRelative
    alphaBoolBlend {blendOption=Linear}
    boundBone {boneName=dyn_necklace_l_01}
    componentAppliedLinearAccClamp=<10000 10000 10000>
    componentLinearAccScale=<0.5 0.5 0.5>
    componentLinearVelScale=<0.5 0.5 0.5>
    gravityOverride {z=-350}
    physicsBodyDefinitions [{
      boundBone {boneName=dyn_necklace_l_01}
      boxExtents=<17 17 17>
      constraintSetup {angularLimitsMax=<0 0 5>, angularLimitsMin=<0 -30 -5>, angularTarget=<1 0 0>}
      localJointOffset=<2 0 0>
     }]
    relativeSpaceBone {boneName=root}
   }
   n24 = Anim Dynamics - Bone: dyn_necklace_r_01!(ComponentPose=@n23, Alpha=1, GravityOverride:Vector="0,0,-350", ExternalForce="0,0,0") -> (Pose) {
    angularDampingOverride=0.89
    angularSpringConstant=100
    bAngularSpring=true
    bGravityOverrideInSimSpace=true
    bOverrideAngularDamping=true
    bUseGravityOverride=true
    linearDampingOverride=0.9
    linearSpringConstant=150
    numSolverIterationsPreUpdate=3
    simulationSpace=BoneRelative
    alphaBoolBlend {blendOption=Linear}
    boundBone {boneName=dyn_necklace_r_01}
    componentAppliedLinearAccClamp=<10000 10000 10000>
    componentLinearAccScale=<0.5 0.5 0.5>
    componentLinearVelScale=<0.5 0.5 0.5>
    gravityOverride {z=-350}
    physicsBodyDefinitions [{
      boundBone {boneName=dyn_necklace_r_01}
      boxExtents=<17 17 17>
      constraintSetup {angularLimitsMax=<0 0 5>, angularLimitsMin=<0 -30 -5>, angularTarget=<1 0 0>}
      localJointOffset=<2 0 0>
     }]
    relativeSpaceBone {boneName=root}
   }
   n25 = Anim Dynamics - Bone: dyn_necklace_l_02!(ComponentPose=@n24, Alpha=1, GravityOverride:Vector="0,0,-350", ExternalForce="0,0,0") -> (Pose) {
    angularDampingOverride=0.89
    angularSpringConstant=100
    bAngularSpring=true
    bGravityOverrideInSimSpace=true
    bOverrideAngularDamping=true
    bUseGravityOverride=true
    linearDampingOverride=0.9
    linearSpringConstant=150
    numSolverIterationsPreUpdate=3
    simulationSpace=BoneRelative
    alphaBoolBlend {blendOption=Linear}
    boundBone {boneName=dyn_necklace_l_02}
    componentAppliedLinearAccClamp=<10000 10000 10000>
    componentLinearAccScale=<0.5 0.5 0.5>
    componentLinearVelScale=<0.5 0.5 0.5>
    gravityOverride {z=-350}
    physicsBodyDefinitions [{
      boundBone {boneName=dyn_necklace_l_02}
      boxExtents=<17 17 17>
      constraintSetup {angularLimitsMax=<0 0 5>, angularLimitsMin=<0 -60 -5>, angularTarget=<1 0 0>}
      localJointOffset=<2 0 0>
     }]
    relativeSpaceBone {boneName=root}
   }
   n26 = Anim Dynamics - Bone: dyn_necklace_r_02!(ComponentPose=@n25, Alpha=1, GravityOverride:Vector="0,0,-350", ExternalForce="0,0,0") -> (Pose) {
    angularDampingOverride=0.89
    angularSpringConstant=100
    bAngularSpring=true
    bGravityOverrideInSimSpace=true
    bOverrideAngularDamping=true
    bUseGravityOverride=true
    linearDampingOverride=0.9
    linearSpringConstant=150
    numSolverIterationsPreUpdate=3
    simulationSpace=BoneRelative
    alphaBoolBlend {blendOption=Linear}
    boundBone {boneName=dyn_necklace_r_02}
    componentAppliedLinearAccClamp=<10000 10000 10000>
    componentLinearAccScale=<0.5 0.5 0.5>
    componentLinearVelScale=<0.5 0.5 0.5>
    gravityOverride {z=-350}
    physicsBodyDefinitions [{
      boundBone {boneName=dyn_necklace_r_02}
      boxExtents=<17 17 17>
      constraintSetup {angularLimitsMax=<0 0 5>, angularLimitsMin=<0 -60 -5>, angularTarget=<1 0 0>}
      localJointOffset=<2 0 0>
     }]
    relativeSpaceBone {boneName=root}
   }
   n27 = Anim Dynamics - Bone: dyn_necklace_l_03!(ComponentPose=@n26, Alpha=1, GravityOverride:Vector="0,0,-350", ExternalForce="0,0,0") -> (Pose) {
    angularDampingOverride=0.89
    angularSpringConstant=100
    bAngularSpring=true
    bGravityOverrideInSimSpace=true
    bOverrideAngularDamping=true
    bUseGravityOverride=true
    linearDampingOverride=0.9
    linearSpringConstant=150
    numSolverIterationsPreUpdate=3
    simulationSpace=BoneRelative
    alphaBoolBlend {blendOption=Linear}
    boundBone {boneName=dyn_necklace_l_03}
    componentAppliedLinearAccClamp=<10000 10000 10000>
    componentLinearAccScale=<0.5 0.5 0.5>
    componentLinearVelScale=<0.5 0.5 0.5>
    gravityOverride {z=-350}
    physicsBodyDefinitions [{
      boundBone {boneName=dyn_necklace_l_03}
      boxExtents=<17 17 17>
      constraintSetup {angularLimitsMax=<0 0 5>, angularLimitsMin=<0 -60 -5>, angularTarget=<1 0 0>}
      localJointOffset=<2 0 0>
     }]
    relativeSpaceBone {boneName=root}
   }
   n28 = Anim Dynamics - Bone: dyn_necklace_r_03!(ComponentPose=@n27, Alpha=1, GravityOverride:Vector="0,0,-350", ExternalForce="0,0,0") -> (Pose) {
    angularDampingOverride=0.89
    angularSpringConstant=100
    bAngularSpring=true
    bGravityOverrideInSimSpace=true
    bOverrideAngularDamping=true
    bUseGravityOverride=true
    linearDampingOverride=0.9
    linearSpringConstant=150
    numSolverIterationsPreUpdate=3
    simulationSpace=BoneRelative
    alphaBoolBlend {blendOption=Linear}
    boundBone {boneName=dyn_necklace_r_03}
    componentAppliedLinearAccClamp=<10000 10000 10000>
    componentLinearAccScale=<0.5 0.5 0.5>
    componentLinearVelScale=<0.5 0.5 0.5>
    gravityOverride {z=-350}
    physicsBodyDefinitions [{
      boundBone {boneName=dyn_necklace_r_03}
      boxExtents=<17 17 17>
      constraintSetup {angularLimitsMax=<0 0 5>, angularLimitsMin=<0 -60 -5>, angularTarget=<1 0 0>}
      localJointOffset=<2 0 0>
     }]
    relativeSpaceBone {boneName=root}
   }
   n29 = Anim Dynamics - Bone: dyn_necklace_l_04!(ComponentPose=@n28, Alpha=1, GravityOverride:Vector="0,0,-350", ExternalForce="0,0,0") -> (Pose) {
    angularDampingOverride=0.89
    angularSpringConstant=100
    bAngularSpring=true
    bGravityOverrideInSimSpace=true
    bOverrideAngularDamping=true
    bUseGravityOverride=true
    linearDampingOverride=0.9
    linearSpringConstant=150
    numSolverIterationsPreUpdate=3
    simulationSpace=BoneRelative
    alphaBoolBlend {blendOption=Linear}
    boundBone {boneName=dyn_necklace_l_04}
    componentAppliedLinearAccClamp=<10000 10000 10000>
    componentLinearAccScale=<0.5 0.5 0.5>
    componentLinearVelScale=<0.5 0.5 0.5>
    gravityOverride {z=-350}
    physicsBodyDefinitions [{
      boundBone {boneName=dyn_necklace_l_04}
      boxExtents=<17 17 17>
      constraintSetup {angularLimitsMax=<0 0 5>, angularLimitsMin=<0 -60 -5>, angularTarget=<1 0 0>}
      localJointOffset=<2 0 0>
     }]
    relativeSpaceBone {boneName=root}
   }
   n30 = Anim Dynamics - Bone: dyn_necklace_r_04!(ComponentPose=@n29, Alpha=1, GravityOverride:Vector="0,0,-350", ExternalForce="0,0,0") -> (Pose) {
    angularDampingOverride=0.89
    angularSpringConstant=100
    bAngularSpring=true
    bGravityOverrideInSimSpace=true
    bOverrideAngularDamping=true
    bUseGravityOverride=true
    linearDampingOverride=0.9
    linearSpringConstant=150
    numSolverIterationsPreUpdate=3
    simulationSpace=BoneRelative
    alphaBoolBlend {blendOption=Linear}
    boundBone {boneName=dyn_necklace_r_04}
    componentAppliedLinearAccClamp=<10000 10000 10000>
    componentLinearAccScale=<0.5 0.5 0.5>
    componentLinearVelScale=<0.5 0.5 0.5>
    gravityOverride {z=-350}
    physicsBodyDefinitions [{
      boundBone {boneName=dyn_necklace_r_04}
      boxExtents=<17 17 17>
      constraintSetup {angularLimitsMax=<0 0 5>, angularLimitsMin=<0 -60 -5>, angularTarget=<1 0 0>}
      localJointOffset=<2 0 0>
     }]
    relativeSpaceBone {boneName=root}
   }
   n31 = Anim Dynamics - Bone: dyn_necklace_c!(ComponentPose=@n30, Alpha=1, GravityOverride:Vector="0,0,-350", ExternalForce="0,0,0") -> (Pose) {
    angularDampingOverride=0.89
    angularSpringConstant=100
    bAngularSpring=true
    bGravityOverrideInSimSpace=true
    bOverrideAngularDamping=true
    bUseGravityOverride=true
    linearDampingOverride=0.9
    linearSpringConstant=150
    numSolverIterationsPreUpdate=3
    simulationSpace=BoneRelative
    alphaBoolBlend {blendOption=Linear}
    boundBone {boneName=dyn_necklace_c}
    componentAppliedLinearAccClamp=<10000 10000 10000>
    componentLinearAccScale=<0.5 0.5 0.5>
    componentLinearVelScale=<0.5 0.5 0.5>
    gravityOverride {z=-350}
    physicsBodyDefinitions [{
      boundBone {boneName=dyn_necklace_c}
      boxExtents=<17 17 17>
      constraintSetup {angularLimitsMax=<0 0 5>, angularLimitsMin=<0 -30 -5>, angularTarget=<1 0 0>}
      localJointOffset=<2 0 0>
     }]
    relativeSpaceBone {boneName=root}
   }
   n32 = "Constraint - dyn_necklace_c (PPP)"!(ComponentPose=@n31, Alpha=1, ConstraintWeights_0=0.4, ConstraintWeights_1:float=0.4, ConstraintWeights_2:float=0.2) -> (Pose) {
    alphaBoolBlend {blendOption=Linear}
    boneToModify {boneName=dyn_necklace_c}
    constraintSetup [{targetBone {boneName=dyn_necklace_r_04}}, {targetBone {boneName=dyn_necklace_l_04}}, {targetBone {boneName=spine_03}}]
    constraintWeights [1,1,1]
   }
   n33 = Anim Dynamics - Bone: dyn_bracelet(ComponentPose=@n32, Alpha=1, ExternalForce="0,0,0") {
    angularDampingOverride=0.89
    angularSpringConstant=100
    bGravityOverrideInSimSpace=true
    bLinearSpring=true
    bOverrideAngularDamping=true
    bOverrideLinearDamping=true
    bUseGravityOverride=true
    linearDampingOverride=0.9999
    linearSpringConstant=15
    numSolverIterationsPreUpdate=3
    simulationSpace=BoneRelative
    alphaBoolBlend {blendOption=Linear}
    boundBone {boneName=dyn_bracelet}
    componentAppliedLinearAccClamp=<10000 10000 10000>
    componentLinearAccScale=<1 1 1>
    componentLinearVelScale=<1 1 1>
    gravityOverride {z=980}
    physicsBodyDefinitions [{
      boundBone {boneName=dyn_bracelet}
      boxExtents=<75 75 75>
      constraintSetup {angularLimitsMax=<0 0 0>, angularLimitsMin=<0 0 0>, angularTarget=<1 0 0>, linearAxesMax=<1 0 0>, linearAxesMin=<-1 0 0>}
      localJointOffset=<2 0 0>
     }]
    relativeSpaceBone {boneName=lowerarm_l}
   }
   Output Pose(Result=Component To Local(ComponentPose=@n33)) {layerGroup=None}
  ]}
 EventGraph {nodes [
   n1 = Event BlueprintUpdateAnimation()
   > n2 = Sequence!(execute) -> (then_0, then_1, then_2:Exec, then_3:Exec, then_4:Exec, then_5:Exec, then_6:Exec, then_7:Exec, then_8:Exec)
   n3 = Get bIsSkyDiving()
   n4 = Branch(execute=@n2.then_1, Condition=@n3)
   > n5 = Set ClothSim_WindStrength(ClothSim_WindStrength:double=0.4) -> (Output_Get:double)
   n6 = Get Mat Instance Dynamic()
   SetScalarParameterValue(
    execute=[@n7.then, @n5.then]
    self=@n6
    ParameterName=ClothSim_WindStrength
    Value=Get ClothSim_WindStrength() -> (ClothSim_WindStrength:double)
   )
   n7 = Set ClothSim_WindStrength(execute=@n4.else, ClothSim_WindStrength=0)
   n8 = Branch(execute=@n2.then_2, Condition=@n3)
   > n9 = Set ClothSim_VelocityMax(ClothSim_VelocityMax:double=750) -> (Output_Get:double)
   SetScalarParameterValue(
    execute=[@n10.then, @n9.then]
    self=@n6
    ParameterName=ClothSim_VelocityMax
    Value=Get ClothSim_VelocityMax() -> (ClothSim_VelocityMax:double)
   )
   n10 = Set ClothSim_VelocityMax(execute=@n8.else, ClothSim_VelocityMax=5000)
   n11 = Branch(execute=@n2.then_3, Condition=@n3)
   > n12 = Set ClothSim_WindNoiseSpeed(ClothSim_WindNoiseSpeed:double=200) -> (Output_Get:double)
   SetScalarParameterValue(
    execute=[@n13.then, @n12.then]
    self=@n6
    ParameterName=ClothSim_WindNoiseSpeed
    Value=Get ClothSim_WindNoiseSpeed() -> (ClothSim_WindNoiseSpeed:double)
   )
   n13 = Set ClothSim_WindNoiseSpeed(execute=@n11.else, ClothSim_WindNoiseSpeed=0)
   n14 = PlayerMovement(execute=@n2.then_5, AnimInstance=Self-Reference(), XDotProductCoefficient=1, YDotProductCoefficient=1, ZDotProductCoefficient=1)
   > n15 = Branch(Condition="Not Equal (Float)"(A=float + float!(A=@n14, B=@n14.YMovement, C:float=@n14.ZMovement) -> (ReturnValue)))
   > Set movementSum(movementSum:bool=true) -> (Output_Get:bool)
   Set movementSum(execute=@n15.else, movementSum=false)
   n16 = Branch(execute=@n2.then_6, Condition=Get movementSum() -> (movementSum:bool))
   SetVectorParameterValue(
    execute=@n16.else
    self=Get Mat Instance Dynamic()
    ParameterName=ClothSim_ABP_R1_Velocity
    Value=Get Value() -> (Value:Linear Color Structure)
   )
   n17 = Gravity Override State Function DEPRECATED(
    execute=@n2.then_0
    CharacterPartAnimInstance=Self-Reference()
    GameTimeInSeconds=GetGameTimeInSeconds()
    GravityOverrideParams=Get GravityOverrideParams() -> (GravityOverrideParams:Gravity Override Params Struct Structure)
   )
   > Set GravityOverride(GravityOverride=@n17)
   n18 = Get!(execute:Exec=@n2.then_4) -> (then:Exec, else:Exec, skeletal_mesh:Skeletal Mesh Component Object Reference)
   > Set WPO Globals Function(
    Material=Get Mat Instance Dynamic()
    Overlay Material=Get Overlay Mat Instance Dynamic()
    Skeletal Mesh=@n18
    "Local Noise Transform  Location"=Get Local Noise Transform Location() -> (Local Noise Transform Location:Vector)
    Local Noise Offset=Get Local Noise Offset()
   )
   > Set WPO Regional Function!(
    execute
    Material:Material Instance Dynamic Object Reference=Get Mat Instance Dynamic()
    Overlay Material:Material Instance Dynamic Object Reference=Get Overlay Mat Instance Dynamic()
    Skeletal Mesh:Skeletal Mesh Component Object Reference=Get skeletal_mesh() -> (skeletal_mesh:Skeletal Mesh Component Object Reference)
    Velocity Transform:Name=pelvis
    Velocity Interp Speed:double=Get Velocity Interp Speed()
    Collision Transform:Name=spine_01
    Collision Transform Axis:EAxis Enum
    Invert Collision Transform Axis:bool=false
    Delta Time X:double=@n1.DeltaTimeX
    Previous Velocity Transform Location:"Vector (by ref)"=Get R1 Velocity Transform Location()
    Previous Velocity Transform Velocity:"Vector (by ref)"=Get R1 Velocity Transform Velocity()
    Region:int=1
   ) -> (then)
   n19 = GetCurveValue(self=GetOwnerFrontEndAnimInstance(), CurveName=Frontend_Idle_Bypass)
   n20 = Branch(execute=@n2.then_7, Condition=AND Boolean(A=Get bIsInFrontEnd(), B="float > float"(A=@n19)))
   > Set customIdleDynNecklaceAlpha(customIdleDynNecklaceAlpha:float=0.3) -> (Output_Get:float)
   > Set customIdleDynBraceletAlpha(customIdleDynBraceletAlpha:double=0) -> (Output_Get:double)
   Set customIdleDynNecklaceAlpha(execute=@n20.else, customIdleDynNecklaceAlpha=1)
   > Set customIdleDynBraceletAlpha(customIdleDynBraceletAlpha=1)
   Event OnCharacterCustomizationLoaded()
   > n21 = Set skeletal_mesh(skeletal_mesh:Skeletal Mesh Component Object Reference=GetPartSkeletalMeshComponent()) -> (Output_Get:Skeletal Mesh Component Object Reference)
   > n22 = Get Dynamic Material Instance(SkeletalMeshComponent=@n21)
   > Set Mat Instance Dynamic(Mat Instance Dynamic=@n22)
   > n23 = GetOrCreateDynamicOverlayMaterial(MeshComponent=@n21, MaterialName=None)
   > Set Overlay Mat Instance Dynamic(Overlay Mat Instance Dynamic=@n23)
  ]}
]
variables [
 ClothSim_VelocityMax {type=number}
 ClothSim_WindNoiseSpeed {type=number}
 ClothSim_WindStrength {type=number}
 customIdleDynBraceletAlpha {type=number}
 customIdleDynNecklaceAlpha {type=number}
 GravityOverride {type=Vector}
 GravityOverrideParams {type=GravityOverrideParamsStruct}
 isCustomEmotePlaying {type=number}
 isCustomEmotePlaying_bool {type=boolean}
 isNotFrontend {type=boolean}
 Local Noise Offset {type=Vector}
 Local Noise Transform Location {type=Vector}
 Mat Instance Dynamic {type=MaterialInstanceDynamic}
 movementSum {type=boolean}
 Overlay Mat Instance Dynamic {type=MaterialInstanceDynamic}
 ownerSkelMesh {type=SkeletalMeshComponent}
 PlayerPawn {type=FortPlayerPawnAthena}
 R1 Velocity Transform Location {type=Vector}
 R1 Velocity Transform Velocity {type=Vector}
 RigidBodyClothOffANDcustomEmote {type=boolean}
 skeletal_mesh {type=SkeletalMeshComponent}
 Value {type=LinearColor, default=<rgba 0 0 0 0>}
 Velocity Interp Speed {type=number}
]
